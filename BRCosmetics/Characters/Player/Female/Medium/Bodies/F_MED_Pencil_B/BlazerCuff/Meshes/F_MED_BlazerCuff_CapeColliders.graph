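# Physics asset for the BlazerCuff female medium body, named as the collision set that capes collide against.
# Capsules cover the torso, clavicles, head and feet, while arm, thigh, calf and neck bodies have no shapes.
# Capsule chains on dynamic bones add garment volume: eight skirt-like cloth rings' columns from the pelvis, five jacket chains from spine_04.
# Constraints follow these chains with all angular motion free.
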PhysicsAsset /BRCosmetics/Characters/Player/Female/Medium/Bodies/F_MED_Pencil_B/BlazerCuff/Meshes/F_MED_BlazerCuff_CapeColliders
bodies {
 calf_l []
 calf_r []
 clavicle_l [{shapeType=Capsule, length=6.5623, radius=7.5194, center=<3.9566 -1.0862 -2.1127>, rotation=<rot -77.5566 177.3915 7.7314>}]
 clavicle_r [{shapeType=Capsule, length=6.5623, radius=7.5194, center=<-3.9566 1.0862 2.1127>, rotation=<rot -77.5566 177.3915 -172.2686>}]
 dyn_Cloth_volume_bk [{shapeType=Capsule, length=12, radius=10, shapeName=dyn_Cloth_volume_bk_capsule, center=<-15 3 0>, rotation=<rot 90 0 0>}]
 dyn_Cloth_volume_bk01 [{shapeType=Capsule, length=12, radius=10, shapeName=dyn_Cloth_volume_bk01_capsule, center=<-15 3 0>, rotation=<rot 90 0 0>}]
 dyn_Cloth_volume_bk02 [{shapeType=Capsule, length=12, radius=10, shapeName=dyn_Cloth_volume_bk02_capsule, center=<-15 -3.0223 0>, rotation=<rot 90 0 -10>}]
 dyn_Cloth_volume_bk03 [{shapeType=Capsule, length=6.5, radius=15.5, shapeName=dyn_Cloth_volume_bk03_capsule, center=<-15 -2.0996 0>, rotation=<rot 90 0 -5>}]
 dyn_Cloth_volume_fn [{shapeType=Capsule, length=12, radius=10, shapeName=dyn_Cloth_volume_fn_capsule, center=<-15 3 0>, rotation=<rot 90 0 0>}]
 dyn_Cloth_volume_fn01 [{shapeType=Capsule, length=12, radius=10, shapeName=dyn_Cloth_volume_fn01_capsule, center=<-15 3 0>, rotation=<rot 90 0 0>}]
 dyn_Cloth_volume_fn02 [{shapeType=Capsule, length=12, radius=10, shapeName=dyn_Cloth_volume_fn02_capsule, center=<-15 3 0>, rotation=<rot 90 0 0>}]
 dyn_Cloth_volume_fn03 [{shapeType=Capsule, length=12, radius=10, shapeName=dyn_Cloth_volume_fn03_capsule, center=<-15 3 0>, rotation=<rot 90 0 0>}]
 dyn_Cloth_volume_l_A [{shapeType=Capsule, length=12, radius=10, shapeName=dyn_Cloth_volume_l_A_capsule, center=<-15 3 0>, rotation=<rot 90 0 0>}]
 dyn_Cloth_volume_l_A01 [{shapeType=Capsule, length=12, radius=10, shapeName=dyn_Cloth_volume_l_A01_capsule, center=<-15 3 0>, rotation=<rot 90 0 0>}]
 dyn_Cloth_volume_l_A02 [{shapeType=Capsule, length=12, radius=10, shapeName=dyn_Cloth_volume_l_A02_capsule, center=<-15 3 0>, rotation=<rot 90 0 0>}]
 dyn_Cloth_volume_l_A03 [{shapeType=Capsule, length=12, radius=10, shapeName=dyn_Cloth_volume_l_A03_capsule, center=<-15 3 0>, rotation=<rot 90 0 0>}]
 dyn_Cloth_volume_l_B [{shapeType=Capsule, length=12, radius=10, shapeName=dyn_Cloth_volume_l_B_capsule, center=<-15 3 0>, rotation=<rot 90 0 0>}]
 dyn_Cloth_volume_l_B01 [{shapeType=Capsule, length=12, radius=10, shapeName=dyn_Cloth_volume_l_B01_capsule, center=<-15 3 0>, rotation=<rot 90 0 0>}]
 dyn_Cloth_volume_l_B02 [{shapeType=Capsule, length=12, radius=10, shapeName=dyn_Cloth_volume_l_B02_capsule, center=<-15 3 0>, rotation=<rot 90 0 0>}]
 dyn_Cloth_volume_l_B03 [{shapeType=Capsule, length=12, radius=10, shapeName=dyn_Cloth_volume_l_B03_capsule, center=<-15 3 0>, rotation=<rot 90 0 0>}]
 dyn_Cloth_volume_l_C [{shapeType=Capsule, length=12, radius=10, shapeName=dyn_Cloth_volume_l_C_capsule, center=<-15 3 0>, rotation=<rot 90 0 0>}]
 dyn_Cloth_volume_l_C01 [{shapeType=Capsule, length=12, radius=10, shapeName=dyn_Cloth_volume_l_C01_capsule, center=<-15 3 0>, rotation=<rot 90 0 0>}]
 dyn_Cloth_volume_l_C02 [{shapeType=Capsule, length=12, radius=10, shapeName=dyn_Cloth_volume_l_C02_capsule, center=<-15 -3.0223 0>, rotation=<rot 90 180 170>}]
 dyn_Cloth_volume_l_C03 [{shapeType=Capsule, length=6.5, radius=15.5, shapeName=dyn_Cloth_volume_l_C03_capsule, center=<-15 -2.0996 0>, rotation=<rot 90 180 175>}]
 dyn_Cloth_volume_r_A [{shapeType=Capsule, length=12, radius=10, shapeName=dyn_Cloth_volume_r_A_capsule, center=<-15 3 0>, rotation=<rot 90 0 0>}]
 dyn_Cloth_volume_r_A01 [{shapeType=Capsule, length=12, radius=10, shapeName=dyn_Cloth_volume_r_A01_capsule, center=<-15 3 0>, rotation=<rot 90 0 0>}]
 dyn_Cloth_volume_r_A02 [{shapeType=Capsule, length=12, radius=10, shapeName=dyn_Cloth_volume_r_A02_capsule, center=<-15 3 0>, rotation=<rot 90 0 0>}]
 dyn_Cloth_volume_r_A03 [{shapeType=Capsule, length=12, radius=10, shapeName=dyn_Cloth_volume_r_A03_capsule, center=<-15 3 0>, rotation=<rot 90 0 0>}]
 dyn_Cloth_volume_r_B [{shapeType=Capsule, length=12, radius=10, shapeName=dyn_Cloth_volume_r_B_capsule, center=<-15 3 0>, rotation=<rot 90 0 0>}]
 dyn_Cloth_volume_r_B01 [{shapeType=Capsule, length=12, radius=10, shapeName=dyn_Cloth_volume_r_B01_capsule, center=<-15 3 0>, rotation=<rot 90 0 0>}]
 dyn_Cloth_volume_r_B02 [{shapeType=Capsule, length=12, radius=10, shapeName=dyn_Cloth_volume_r_B02_capsule, center=<-15 3 0>, rotation=<rot 90 0 0>}]
 dyn_Cloth_volume_r_B03 [{shapeType=Capsule, length=12, radius=10, shapeName=dyn_Cloth_volume_r_B03_capsule, center=<-15 3 0>, rotation=<rot 90 0 0>}]
 dyn_Cloth_volume_r_C [{shapeType=Capsule, length=12, radius=10, shapeName=dyn_Cloth_volume_r_C_capsule, center=<-15 3 0>, rotation=<rot 90 0 0>}]
 dyn_Cloth_volume_r_C01 [{shapeType=Capsule, length=12, radius=10, shapeName=dyn_Cloth_volume_r_C01_capsule, center=<-15 3 0>, rotation=<rot 90 0 0>}]
 dyn_Cloth_volume_r_C02 [{
   shapeType=Capsule
   length=12
   radius=10
   shapeName=dyn_Cloth_volume_r_C02_capsule
   center=<-15 -3.0223 0>
   rotation=<rot 90 -1.4765 -11.4765>
  }]
 dyn_Cloth_volume_r_C03 [{
   shapeType=Capsule
   length=6.5
   radius=15.5
   shapeName=dyn_Cloth_volume_r_C03_capsule
   center=<-15 -2.0996 0>
   rotation=<rot 90 -1.4765 -6.4765>
  }]
 dyn_jacket_bk_01 [{shapeType=Capsule, length=8, radius=6, shapeName=dyn_jacket_bk_01_capsule, center=<5 5 0>, rotation=<rot 90 0 0>}]
 dyn_jacket_bk_02 [{shapeType=Capsule, length=8, radius=6, shapeName=dyn_jacket_bk_02_capsule, center=<5 5 0>, rotation=<rot 90 0 0>}]
 dyn_jacket_bk_03 [{shapeType=Capsule, length=8, radius=6, shapeName=dyn_jacket_bk_03_capsule, center=<5 5 0>, rotation=<rot 90 0 0>}]
 dyn_jacket_bk_l_01 [{shapeType=Capsule, length=8, radius=6, shapeName=dyn_jacket_bk_l_01_capsule, center=<5 0 -5>, rotation=<rot 90 0 0>}]
 dyn_jacket_bk_l_02 [{shapeType=Capsule, length=8, radius=6, shapeName=dyn_jacket_bk_l_02_capsule, center=<5 0 -5>, rotation=<rot 90 0 0>}]
 dyn_jacket_bk_l_03 [{shapeType=Capsule, length=8, radius=6, shapeName=dyn_jacket_bk_l_03_capsule, center=<5 0 -5>, rotation=<rot 90 0 0>}]
 dyn_jacket_bk_r_01 [{shapeType=Capsule, length=8, radius=6, shapeName=dyn_jacket_bk_r_01_capsule, center=<5 5 0>, rotation=<rot 90 0 0>}]
 dyn_jacket_bk_r_02 [{shapeType=Capsule, length=8, radius=6, shapeName=dyn_jacket_bk_r_02_capsule, center=<5 5 0>, rotation=<rot 90 0 0>}]
 dyn_jacket_bk_r_03 [{shapeType=Capsule, length=8, radius=6, shapeName=dyn_jacket_bk_r_03_capsule, center=<5 5 0>, rotation=<rot 90 0 0>}]
 dyn_jacket_l_01 [{shapeType=Capsule, length=8, radius=6, shapeName=dyn_jacket_l_01_capsule, center=<5 0 -5>, rotation=<rot 90 0 0>}]
 dyn_jacket_l_02 [{shapeType=Capsule, length=8, radius=6, shapeName=dyn_jacket_l_02_capsule, center=<5 0 -5>, rotation=<rot 90 0 0>}]
 dyn_jacket_l_03 [{shapeType=Capsule, length=8, radius=6, shapeName=dyn_jacket_l_03_capsule, center=<5 0 -5>, rotation=<rot 90 0 0>}]
 dyn_jacket_r_01 [{shapeType=Capsule, length=8, radius=6, shapeName=dyn_jacket_r_01_capsule, center=<5 5 0>, rotation=<rot 90 0 0>}]
 dyn_jacket_r_02 [{shapeType=Capsule, length=8, radius=6, shapeName=dyn_jacket_r_02_capsule, center=<5 5 0>, rotation=<rot 90 0 0>}]
 dyn_jacket_r_03 [{shapeType=Capsule, length=8, radius=6, shapeName=dyn_jacket_r_03_capsule, center=<5 5 0>, rotation=<rot 90 0 0>}]
 foot_l [{shapeType=Capsule, length=11.0765, radius=8.0344, center=<-4.1811 6.0027 -0.338>, rotation=<rot 0.0388 1.4701 -89.3982>}]
 foot_r [{shapeType=Capsule, length=11.0765, radius=8.0344, center=<4.1811 -6.0027 0.338>, rotation=<rot 0.0388 1.4701 90.6018>}]
 head [{shapeType=Capsule, length=0, radius=18, center=<9.0697 0.0746 0>, rotation=<rot 0 -90 0>}]
 lowerarm_l []
 lowerarm_r []
 neck_02 []
 pelvis [{shapeType=Capsule, length=8.6042, radius=12.1441, center=<-2.4471 -2.5289 -0.1981>}]
 spine_01 [{shapeType=Capsule, length=3.2638, radius=15.4888, center=<-5.0346 2.6641 0>}]
 spine_02 [
  {shapeType=Capsule, length=2.625, radius=12.4817, center=<1.179 2.2974 0>}
  {shapeType=Capsule, length=18.8125, radius=11.1875, center=<-1.8317 1.9362 0>, rotation=<rot 90 0 0>}
 ]
 spine_03 [
  {shapeType=Capsule, length=6.1863, radius=9.9854, center=<0.0574 0.4416 0>}
  {shapeType=Capsule, length=19.8438, radius=10.1562, center=<-1.2163 0.549 0>, rotation=<rot 90 -178.8755 -178.9484>}
 ]
 spine_04 [
  {shapeType=Capsule, length=9.0076, radius=10.3916, center=<5.6234 -0.6779 0>}
  {shapeType=Capsule, length=9.0312, radius=10.2812, center=<-1.2326 1.308 0>, rotation=<rot 90 0 0>}
 ]
 spine_05 [{shapeType=Capsule, length=9.678, radius=12.3571, center=<0.6969 1.91 0>}]
 thigh_l []
 thigh_r []
 upperarm_l []
 upperarm_r []
}
constraints [
 {
  bone1Name=spine_01
  bone2Name=pelvis
  swing1LimitDegrees=45
  swing1Motion=Free
  swing2LimitDegrees=45
  swing2Motion=Free
  twistLimitDegrees=45
  twistMotion=Free
 }
 {
  bone1Name=spine_02
  bone2Name=spine_01
  swing1LimitDegrees=45
  swing1Motion=Free
  swing2LimitDegrees=45
  swing2Motion=Free
  twistLimitDegrees=45
  twistMotion=Free
 }
 {
  bone1Name=spine_03
  bone2Name=spine_02
  swing1LimitDegrees=45
  swing1Motion=Free
  swing2LimitDegrees=45
  swing2Motion=Free
  twistLimitDegrees=45
  twistMotion=Free
 }
 {
  bone1Name=spine_04
  bone2Name=spine_03
  swing1LimitDegrees=45
  swing1Motion=Free
  swing2LimitDegrees=45
  swing2Motion=Free
  twistLimitDegrees=45
  twistMotion=Free
 }
 {
  bone1Name=spine_05
  bone2Name=spine_04
  swing1LimitDegrees=45
  swing1Motion=Free
  swing2LimitDegrees=45
  swing2Motion=Free
  twistLimitDegrees=45
  twistMotion=Free
 }
 {
  bone1Name=clavicle_l
  bone2Name=spine_05
  swing1LimitDegrees=45
  swing1Motion=Free
  swing2LimitDegrees=45
  swing2Motion=Free
  twistLimitDegrees=45
  twistMotion=Free
 }
 {
  bone1Name=upperarm_l
  bone2Name=clavicle_l
  swing1LimitDegrees=45
  swing1Motion=Free
  swing2LimitDegrees=45
  swing2Motion=Free
  twistLimitDegrees=45
  twistMotion=Free
 }
 {
  bone1Name=lowerarm_l
  bone2Name=upperarm_l
  swing1LimitDegrees=45
  swing1Motion=Free
  swing2LimitDegrees=45
  swing2Motion=Free
  twistLimitDegrees=45
  twistMotion=Free
 }
 {
  bone1Name=neck_02
  bone2Name=spine_05
  swing1LimitDegrees=45
  swing1Motion=Free
  swing2LimitDegrees=45
  swing2Motion=Free
  twistLimitDegrees=45
  twistMotion=Free
 }
 {
  bone1Name=head
  bone2Name=neck_02
  swing1LimitDegrees=45
  swing1Motion=Free
  swing2LimitDegrees=45
  swing2Motion=Free
  twistLimitDegrees=45
  twistMotion=Free
 }
 {
  bone1Name=calf_l
  bone2Name=thigh_l
  swing1LimitDegrees=45
  swing1Motion=Free
  swing2LimitDegrees=45
  swing2Motion=Free
  twistLimitDegrees=45
  twistMotion=Free
 }
 {
  bone1Name=thigh_l
  bone2Name=pelvis
  swing1LimitDegrees=45
  swing1Motion=Free
  swing2LimitDegrees=45
  swing2Motion=Free
  twistLimitDegrees=45
  twistMotion=Free
 }
 {
  bone1Name=foot_l
  bone2Name=calf_l
  swing1LimitDegrees=45
  swing1Motion=Free
  swing2LimitDegrees=45
  swing2Motion=Free
  twistLimitDegrees=45
  twistMotion=Free
 }
 {
  bone1Name=lowerarm_r
  bone2Name=upperarm_r
  swing1LimitDegrees=45
  swing1Motion=Free
  swing2LimitDegrees=45
  swing2Motion=Free
  twistLimitDegrees=45
  twistMotion=Free
 }
 {
  bone1Name=upperarm_r
  bone2Name=clavicle_r
  swing1LimitDegrees=45
  swing1Motion=Free
  swing2LimitDegrees=45
  swing2Motion=Free
  twistLimitDegrees=45
  twistMotion=Free
 }
 {
  bone1Name=clavicle_r
  bone2Name=spine_05
  swing1LimitDegrees=45
  swing1Motion=Free
  swing2LimitDegrees=45
  swing2Motion=Free
  twistLimitDegrees=45
  twistMotion=Free
 }
 {
  bone1Name=thigh_r
  bone2Name=pelvis
  swing1LimitDegrees=45
  swing1Motion=Free
  swing2LimitDegrees=45
  swing2Motion=Free
  twistLimitDegrees=45
  twistMotion=Free
 }
 {
  bone1Name=foot_r
  bone2Name=calf_r
  swing1LimitDegrees=45
  swing1Motion=Free
  swing2LimitDegrees=45
  swing2Motion=Free
  twistLimitDegrees=45
  twistMotion=Free
 }
 {
  bone1Name=calf_r
  bone2Name=thigh_r
  swing1LimitDegrees=45
  swing1Motion=Free
  swing2LimitDegrees=45
  swing2Motion=Free
  twistLimitDegrees=45
  twistMotion=Free
 }
 {
  bone1Name=dyn_Cloth_volume_bk
  bone2Name=pelvis
  swing1LimitDegrees=45
  swing1Motion=Free
  swing2LimitDegrees=45
  swing2Motion=Free
  twistLimitDegrees=45
  twistMotion=Free
 }
 {
  bone1Name=dyn_Cloth_volume_bk01
  bone2Name=dyn_Cloth_volume_bk
  swing1LimitDegrees=45
  swing1Motion=Free
  swing2LimitDegrees=45
  swing2Motion=Free
  twistLimitDegrees=45
  twistMotion=Free
 }
 {
  bone1Name=dyn_Cloth_volume_bk02
  bone2Name=dyn_Cloth_volume_bk01
  swing1LimitDegrees=45
  swing1Motion=Free
  swing2LimitDegrees=45
  swing2Motion=Free
  twistLimitDegrees=45
  twistMotion=Free
 }
 {
  bone1Name=dyn_Cloth_volume_bk03
  bone2Name=dyn_Cloth_volume_bk02
  swing1LimitDegrees=45
  swing1Motion=Free
  swing2LimitDegrees=45
  swing2Motion=Free
  twistLimitDegrees=45
  twistMotion=Free
 }
 {
  bone1Name=dyn_Cloth_volume_fn
  bone2Name=pelvis
  swing1LimitDegrees=45
  swing1Motion=Free
  swing2LimitDegrees=45
  swing2Motion=Free
  twistLimitDegrees=45
  twistMotion=Free
 }
 {
  bone1Name=dyn_Cloth_volume_fn01
  bone2Name=dyn_Cloth_volume_fn
  swing1LimitDegrees=45
  swing1Motion=Free
  swing2LimitDegrees=45
  swing2Motion=Free
  twistLimitDegrees=45
  twistMotion=Free
 }
 {
  bone1Name=dyn_Cloth_volume_fn02
  bone2Name=dyn_Cloth_volume_fn01
  swing1LimitDegrees=45
  swing1Motion=Free
  swing2LimitDegrees=45
  swing2Motion=Free
  twistLimitDegrees=45
  twistMotion=Free
 }
 {
  bone1Name=dyn_Cloth_volume_fn03
  bone2Name=dyn_Cloth_volume_fn02
  swing1LimitDegrees=45
  swing1Motion=Free
  swing2LimitDegrees=45
  swing2Motion=Free
  twistLimitDegrees=45
  twistMotion=Free
 }
 {
  bone1Name=dyn_Cloth_volume_l_A
  bone2Name=pelvis
  swing1LimitDegrees=45
  swing1Motion=Free
  swing2LimitDegrees=45
  swing2Motion=Free
  twistLimitDegrees=45
  twistMotion=Free
 }
 {
  bone1Name=dyn_Cloth_volume_l_A01
  bone2Name=dyn_Cloth_volume_l_A
  swing1LimitDegrees=45
  swing1Motion=Free
  swing2LimitDegrees=45
  swing2Motion=Free
  twistLimitDegrees=45
  twistMotion=Free
 }
 {
  bone1Name=dyn_Cloth_volume_l_A02
  bone2Name=dyn_Cloth_volume_l_A01
  swing1LimitDegrees=45
  swing1Motion=Free
  swing2LimitDegrees=45
  swing2Motion=Free
  twistLimitDegrees=45
  twistMotion=Free
 }
 {
  bone1Name=dyn_Cloth_volume_l_A03
  bone2Name=dyn_Cloth_volume_l_A02
  swing1LimitDegrees=45
  swing1Motion=Free
  swing2LimitDegrees=45
  swing2Motion=Free
  twistLimitDegrees=45
  twistMotion=Free
 }
 {
  bone1Name=dyn_Cloth_volume_l_B
  bone2Name=pelvis
  swing1LimitDegrees=45
  swing1Motion=Free
  swing2LimitDegrees=45
  swing2Motion=Free
  twistLimitDegrees=45
  twistMotion=Free
 }
 {
  bone1Name=dyn_Cloth_volume_l_B01
  bone2Name=dyn_Cloth_volume_l_B
  swing1LimitDegrees=45
  swing1Motion=Free
  swing2LimitDegrees=45
  swing2Motion=Free
  twistLimitDegrees=45
  twistMotion=Free
 }
 {
  bone1Name=dyn_Cloth_volume_l_B02
  bone2Name=dyn_Cloth_volume_l_B01
  swing1LimitDegrees=45
  swing1Motion=Free
  swing2LimitDegrees=45
  swing2Motion=Free
  twistLimitDegrees=45
  twistMotion=Free
 }
 {
  bone1Name=dyn_Cloth_volume_l_B03
  bone2Name=dyn_Cloth_volume_l_B02
  swing1LimitDegrees=45
  swing1Motion=Free
  swing2LimitDegrees=45
  swing2Motion=Free
  twistLimitDegrees=45
  twistMotion=Free
 }
 {
  bone1Name=dyn_Cloth_volume_l_C
  bone2Name=pelvis
  swing1LimitDegrees=45
  swing1Motion=Free
  swing2LimitDegrees=45
  swing2Motion=Free
  twistLimitDegrees=45
  twistMotion=Free
 }
 {
  bone1Name=dyn_Cloth_volume_l_C01
  bone2Name=dyn_Cloth_volume_l_C
  swing1LimitDegrees=45
  swing1Motion=Free
  swing2LimitDegrees=45
  swing2Motion=Free
  twistLimitDegrees=45
  twistMotion=Free
 }
 {
  bone1Name=dyn_Cloth_volume_l_C02
  bone2Name=dyn_Cloth_volume_l_C01
  swing1LimitDegrees=45
  swing1Motion=Free
  swing2LimitDegrees=45
  swing2Motion=Free
  twistLimitDegrees=45
  twistMotion=Free
 }
 {
  bone1Name=dyn_Cloth_volume_l_C03
  bone2Name=dyn_Cloth_volume_l_C02
  swing1LimitDegrees=45
  swing1Motion=Free
  swing2LimitDegrees=45
  swing2Motion=Free
  twistLimitDegrees=45
  twistMotion=Free
 }
 {
  bone1Name=dyn_Cloth_volume_r_A
  bone2Name=pelvis
  swing1LimitDegrees=45
  swing1Motion=Free
  swing2LimitDegrees=45
  swing2Motion=Free
  twistLimitDegrees=45
  twistMotion=Free
 }
 {
  bone1Name=dyn_Cloth_volume_r_A01
  bone2Name=dyn_Cloth_volume_r_A
  swing1LimitDegrees=45
  swing1Motion=Free
  swing2LimitDegrees=45
  swing2Motion=Free
  twistLimitDegrees=45
  twistMotion=Free
 }
 {
  bone1Name=dyn_Cloth_volume_r_A02
  bone2Name=dyn_Cloth_volume_r_A01
  swing1LimitDegrees=45
  swing1Motion=Free
  swing2LimitDegrees=45
  swing2Motion=Free
  twistLimitDegrees=45
  twistMotion=Free
 }
 {
  bone1Name=dyn_Cloth_volume_r_A03
  bone2Name=dyn_Cloth_volume_r_A02
  swing1LimitDegrees=45
  swing1Motion=Free
  swing2LimitDegrees=45
  swing2Motion=Free
  twistLimitDegrees=45
  twistMotion=Free
 }
 {
  bone1Name=dyn_Cloth_volume_r_B
  bone2Name=pelvis
  swing1LimitDegrees=45
  swing1Motion=Free
  swing2LimitDegrees=45
  swing2Motion=Free
  twistLimitDegrees=45
  twistMotion=Free
 }
 {
  bone1Name=dyn_Cloth_volume_r_B01
  bone2Name=dyn_Cloth_volume_r_B
  swing1LimitDegrees=45
  swing1Motion=Free
  swing2LimitDegrees=45
  swing2Motion=Free
  twistLimitDegrees=45
  twistMotion=Free
 }
 {
  bone1Name=dyn_Cloth_volume_r_B02
  bone2Name=dyn_Cloth_volume_r_B01
  swing1LimitDegrees=45
  swing1Motion=Free
  swing2LimitDegrees=45
  swing2Motion=Free
  twistLimitDegrees=45
  twistMotion=Free
 }
 {
  bone1Name=dyn_Cloth_volume_r_B03
  bone2Name=dyn_Cloth_volume_r_B02
  swing1LimitDegrees=45
  swing1Motion=Free
  swing2LimitDegrees=45
  swing2Motion=Free
  twistLimitDegrees=45
  twistMotion=Free
 }
 {
  bone1Name=dyn_Cloth_volume_r_C
  bone2Name=pelvis
  swing1LimitDegrees=45
  swing1Motion=Free
  swing2LimitDegrees=45
  swing2Motion=Free
  twistLimitDegrees=45
  twistMotion=Free
 }
 {
  bone1Name=dyn_Cloth_volume_r_C01
  bone2Name=dyn_Cloth_volume_r_C
  swing1LimitDegrees=45
  swing1Motion=Free
  swing2LimitDegrees=45
  swing2Motion=Free
  twistLimitDegrees=45
  twistMotion=Free
 }
 {
  bone1Name=dyn_Cloth_volume_r_C02
  bone2Name=dyn_Cloth_volume_r_C01
  swing1LimitDegrees=45
  swing1Motion=Free
  swing2LimitDegrees=45
  swing2Motion=Free
  twistLimitDegrees=45
  twistMotion=Free
 }
 {
  bone1Name=dyn_Cloth_volume_r_C03
  bone2Name=dyn_Cloth_volume_r_C02
  swing1LimitDegrees=45
  swing1Motion=Free
  swing2LimitDegrees=45
  swing2Motion=Free
  twistLimitDegrees=45
  twistMotion=Free
 }
 {
  bone1Name=dyn_jacket_bk_01
  bone2Name=spine_04
  swing1LimitDegrees=45
  swing1Motion=Free
  swing2LimitDegrees=45
  swing2Motion=Free
  twistLimitDegrees=45
  twistMotion=Free
 }
 {
  bone1Name=dyn_jacket_bk_02
  bone2Name=dyn_jacket_bk_01
  swing1LimitDegrees=45
  swing1Motion=Free
  swing2LimitDegrees=45
  swing2Motion=Free
  twistLimitDegrees=45
  twistMotion=Free
 }
 {
  bone1Name=dyn_jacket_bk_03
  bone2Name=dyn_jacket_bk_02
  swing1LimitDegrees=45
  swing1Motion=Free
  swing2LimitDegrees=45
  swing2Motion=Free
  twistLimitDegrees=45
  twistMotion=Free
 }
 {
  bone1Name=dyn_jacket_bk_l_01
  bone2Name=spine_04
  swing1LimitDegrees=45
  swing1Motion=Free
  swing2LimitDegrees=45
  swing2Motion=Free
  twistLimitDegrees=45
  twistMotion=Free
 }
 {
  bone1Name=dyn_jacket_bk_l_02
  bone2Name=dyn_jacket_bk_l_01
  swing1LimitDegrees=45
  swing1Motion=Free
  swing2LimitDegrees=45
  swing2Motion=Free
  twistLimitDegrees=45
  twistMotion=Free
 }
 {
  bone1Name=dyn_jacket_bk_l_03
  bone2Name=dyn_jacket_bk_l_02
  swing1LimitDegrees=45
  swing1Motion=Free
  swing2LimitDegrees=45
  swing2Motion=Free
  twistLimitDegrees=45
  twistMotion=Free
 }
 {
  bone1Name=dyn_jacket_bk_r_01
  bone2Name=spine_04
  swing1LimitDegrees=45
  swing1Motion=Free
  swing2LimitDegrees=45
  swing2Motion=Free
  twistLimitDegrees=45
  twistMotion=Free
 }
 {
  bone1Name=dyn_jacket_bk_r_02
  bone2Name=dyn_jacket_bk_r_01
  swing1LimitDegrees=45
  swing1Motion=Free
  swing2LimitDegrees=45
  swing2Motion=Free
  twistLimitDegrees=45
  twistMotion=Free
 }
 {
  bone1Name=dyn_jacket_bk_r_03
  bone2Name=dyn_jacket_bk_r_02
  swing1LimitDegrees=45
  swing1Motion=Free
  swing2LimitDegrees=45
  swing2Motion=Free
  twistLimitDegrees=45
  twistMotion=Free
 }
 {
  bone1Name=dyn_jacket_l_01
  bone2Name=spine_04
  swing1LimitDegrees=45
  swing1Motion=Free
  swing2LimitDegrees=45
  swing2Motion=Free
  twistLimitDegrees=45
  twistMotion=Free
 }
 {
  bone1Name=dyn_jacket_l_02
  bone2Name=dyn_jacket_l_01
  swing1LimitDegrees=45
  swing1Motion=Free
  swing2LimitDegrees=45
  swing2Motion=Free
  twistLimitDegrees=45
  twistMotion=Free
 }
 {
  bone1Name=dyn_jacket_l_03
  bone2Name=dyn_jacket_l_02
  swing1LimitDegrees=45
  swing1Motion=Free
  swing2LimitDegrees=45
  swing2Motion=Free
  twistLimitDegrees=45
  twistMotion=Free
 }
 {
  bone1Name=dyn_jacket_r_01
  bone2Name=spine_04
  swing1LimitDegrees=45
  swing1Motion=Free
  swing2LimitDegrees=45
  swing2Motion=Free
  twistLimitDegrees=45
  twistMotion=Free
 }
 {
  bone1Name=dyn_jacket_r_02
  bone2Name=dyn_jacket_r_01
  swing1LimitDegrees=45
  swing1Motion=Free
  swing2LimitDegrees=45
  swing2Motion=Free
  twistLimitDegrees=45
  twistMotion=Free
 }
 {
  bone1Name=dyn_jacket_r_03
  bone2Name=dyn_jacket_r_02
  swing1LimitDegrees=45
  swing1Motion=Free
  swing2LimitDegrees=45
  swing2Motion=Free
  twistLimitDegrees=45
  twistMotion=Free
 }
]
