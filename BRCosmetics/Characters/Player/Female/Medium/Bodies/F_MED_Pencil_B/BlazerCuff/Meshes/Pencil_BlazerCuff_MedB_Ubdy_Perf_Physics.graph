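# Physics asset for the female medium Pencil blazer-cuff body mesh, providing capsule bodies and joints for jacket secondary motion.
# Pelvis, spine, thigh and six pelvis-linked cloth-volume capsules surround four jacket chains: front and non-front, left and right.
# Each chain hangs from spine_02 through free Aim and MSTR joints into three or four angle-limited links, tuned asymmetrically.
# One extra free constraint cross-links the left front chain to the other left chain.
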PhysicsAsset /BRCosmetics/Characters/Player/Female/Medium/Bodies/F_MED_Pencil_B/BlazerCuff/Meshes/Pencil_BlazerCuff_MedB_Ubdy_Perf_Physics
bodies {
 dyn_Cloth_volume_l_A [{shapeType=Capsule, length=11.5, radius=8.5, shapeName=dyn_Cloth_volume_l_A_capsule, center=<-10 4 0>, rotation=<rot 75 0.0001 0>}]
 dyn_Cloth_volume_l_B [{shapeType=Capsule, length=10.25, radius=9.75, shapeName=dyn_Cloth_volume_l_B_capsule, center=<-10 4 -1>, rotation=<rot 90 -179 166>}]
 dyn_Cloth_volume_l_C [{shapeType=Capsule, length=12, radius=10, shapeName=dyn_Cloth_volume_l_C_capsule, center=<-12 5 0>, rotation=<rot 90 180 170.0001>}]
 dyn_Cloth_volume_r_A [{shapeType=Capsule, length=11.5, radius=8.5, shapeName=dyn_Cloth_volume_r_A_capsule, center=<-10 4 0>, rotation=<rot 75.0001 -179.9998 180>}]
 dyn_Cloth_volume_r_B [{shapeType=Capsule, length=10.25, radius=9.75, shapeName=dyn_Cloth_volume_r_B_capsule, center=<-10 4 1>, rotation=<rot 90 0.0215 -14.9785>}]
 dyn_Cloth_volume_r_C [{shapeType=Capsule, length=12, radius=10, shapeName=dyn_Cloth_volume_r_C_capsule, center=<-12 5 0>, rotation=<rot 90 -1.4765 -11.4765>}]
 dyn_jacket_fr_l_01 [{shapeType=Capsule, length=5.375, radius=6.625, shapeName=dyn_jacket_fr_l_01_capsule, center=<4 -6 5>, rotation=<rot 0 0 90>}]
 dyn_jacket_fr_l_02 [{shapeType=Capsule, length=8.625, radius=5.375, shapeName=dyn_jacket_fr_l_02_capsule, center=<4 -6 5>, rotation=<rot 0 0 90>}]
 dyn_jacket_fr_l_03 [{shapeType=Capsule, length=15.125, radius=7.75, shapeName=dyn_jacket_fr_l_03_capsule, center=<4 -8 5>, rotation=<rot 0.0001 0 80>}]
 dyn_jacket_fr_l_04 [{shapeType=Capsule, length=11.5, radius=8.375, shapeName=dyn_jacket_fr_l_04_capsule, center=<4 -8 5>, rotation=<rot 0.0001 0 80>}]
 dyn_jacket_fr_l_Aim [{shapeType=Capsule, length=0.505, radius=0.505, shapeName=dyn_jacket_fr_l_Aim_capsule, rotation=<rot 0 -90 0>}]
 dyn_jacket_fr_l_MSTR [{shapeType=Capsule, length=0.505, radius=0.505, shapeName=dyn_jacket_fr_l_MSTR_capsule, rotation=<rot 0 -90 0>}]
 dyn_jacket_fr_r_01 [{shapeType=Capsule, length=5.375, radius=6.625, shapeName=dyn_jacket_fr_r_01_capsule, center=<4 7 4>, rotation=<rot 0.0001 -0.0001 115>}]
 dyn_jacket_fr_r_02 [{shapeType=Capsule, length=6.75, radius=5.25, shapeName=dyn_jacket_fr_r_02_capsule, center=<4 6 5>, rotation=<rot 0 0 90>}]
 dyn_jacket_fr_r_03 [{shapeType=Capsule, length=7.625, radius=6.375, shapeName=dyn_jacket_fr_r_03_capsule, center=<4 6 5>, rotation=<rot 0.0001 -0.0001 95>}]
 dyn_jacket_fr_r_04 [{shapeType=Capsule, length=8, radius=7, shapeName=dyn_jacket_fr_r_04_capsule, center=<4 6 5>, rotation=<rot 0.0001 -0.0001 95>}]
 dyn_jacket_fr_r_Aim [{shapeType=Capsule, length=0.505, radius=0.505, shapeName=dyn_jacket_fr_r_Aim_capsule, rotation=<rot 0 -90 0>}]
 dyn_jacket_fr_r_MSTR [{shapeType=Capsule, length=0.505, radius=0.505, shapeName=dyn_jacket_fr_r_MSTR_capsule, rotation=<rot 0 -90 0>}]
 dyn_jacket_l_01 [{shapeType=Capsule, length=5.375, radius=6.625, shapeName=dyn_jacket_l_01_capsule, center=<4 0 5>, rotation=<rot 0 0 90>}]
 dyn_jacket_l_02 [{shapeType=Capsule, length=12.75, radius=6.625, shapeName=dyn_jacket_l_02_capsule, center=<4 -4 5>, rotation=<rot 0 -9.9999 90>}]
 dyn_jacket_l_03 [{shapeType=Capsule, length=11.125, radius=7.375, shapeName=dyn_jacket_l_03_capsule, center=<4 -4 5>, rotation=<rot 0 -24.9998 89.9999>}]
 dyn_jacket_l_Aim [{shapeType=Capsule, length=0.505, radius=0.505, shapeName=dyn_jacket_l_Aim_capsule}]
 dyn_jacket_l_MSTR [{shapeType=Capsule, length=0.505, radius=0.505, shapeName=dyn_jacket_l_MSTR_capsule}]
 dyn_jacket_r_01 [{shapeType=Capsule, length=5.375, radius=6.625, shapeName=dyn_jacket_r_01_capsule, center=<5 -5 0>}]
 dyn_jacket_r_02 [{shapeType=Capsule, length=5.375, radius=6.625, shapeName=dyn_jacket_r_02_capsule, center=<5 -5 0>, rotation=<rot 9.9999 0 0>}]
 dyn_jacket_r_03 [{shapeType=Capsule, length=5.375, radius=7.375, shapeName=dyn_jacket_r_03_capsule, center=<5 -5 0>, rotation=<rot 24.9998 0 0.0001>}]
 dyn_jacket_r_Aim [{shapeType=Capsule, length=0.505, radius=0.505, shapeName=dyn_jacket_r_Aim_capsule}]
 dyn_jacket_r_MSTR [{shapeType=Capsule, length=0.505, radius=0.505, shapeName=dyn_jacket_r_MSTR_capsule}]
 pelvis [
  {shapeType=Capsule, length=9.3375, radius=9.5551, center=<-0.6285 -1.8567 0>}
  {shapeType=Capsule, length=10.665, radius=8.8452, shapeName=pelvis_capsule, center=<-3.9978 -3.5897 0>}
 ]
 spine_01 [{shapeType=Capsule, length=4.9363, radius=9.6791, center=<0 -0.1396 -0.0001>}]
 spine_02 [{shapeType=Capsule, length=3.3455, radius=9.255, center=<0 0.6676 -0.0001>}]
 thigh_l [{shapeType=Capsule, length=26.625, radius=10.875, center=<-21.6167 -1.5 1.3426>, rotation=<rot 80 179.8043 180>}]
 thigh_r [{shapeType=Capsule, length=26.625, radius=10.875, center=<21.0533 1.5 -1.1849>, rotation=<rot 80 179.9999 180>}]
}
constraints [
 {
  bone1Name=spine_01
  bone2Name=pelvis
  swing1LimitDegrees=45
  swing1Motion=Free
  swing2LimitDegrees=45
  swing2Motion=Free
  twistLimitDegrees=45
  twistMotion=Free
 }
 {
  bone1Name=spine_02
  bone2Name=spine_01
  swing1LimitDegrees=45
  swing1Motion=Free
  swing2LimitDegrees=45
  swing2Motion=Free
  twistLimitDegrees=45
  twistMotion=Free
 }
 {
  bone1Name=thigh_l
  bone2Name=pelvis
  swing1LimitDegrees=45
  swing1Motion=Free
  swing2LimitDegrees=45
  swing2Motion=Free
  twistLimitDegrees=45
  twistMotion=Free
 }
 {
  bone1Name=thigh_r
  bone2Name=pelvis
  swing1LimitDegrees=45
  swing1Motion=Free
  swing2LimitDegrees=45
  swing2Motion=Free
  twistLimitDegrees=45
  twistMotion=Free
 }
 {
  bone1Name=dyn_jacket_fr_l_Aim
  bone2Name=spine_02
  swing1LimitDegrees=45
  swing1Motion=Free
  swing2LimitDegrees=45
  swing2Motion=Free
  twistLimitDegrees=45
  twistMotion=Free
 }
 {
  bone1Name=dyn_jacket_fr_l_MSTR
  bone2Name=dyn_jacket_fr_l_Aim
  swing1LimitDegrees=45
  swing1Motion=Free
  swing2LimitDegrees=45
  swing2Motion=Free
  twistLimitDegrees=45
  twistMotion=Free
 }
 {bone1Name=dyn_jacket_fr_l_01, bone2Name=dyn_jacket_fr_l_MSTR, swing1LimitDegrees=15, swing2LimitDegrees=15, twistLimitDegrees=15}
 {bone1Name=dyn_jacket_fr_l_02, bone2Name=dyn_jacket_fr_l_01, swing1LimitDegrees=30, swing2LimitDegrees=40, twistLimitDegrees=20}
 {bone1Name=dyn_jacket_fr_l_03, bone2Name=dyn_jacket_fr_l_02, swing1LimitDegrees=30, swing2LimitDegrees=40, twistLimitDegrees=20}
 {bone1Name=dyn_jacket_fr_l_04, bone2Name=dyn_jacket_fr_l_03, swing1LimitDegrees=30, swing2LimitDegrees=40, twistLimitDegrees=20}
 {
  bone1Name=dyn_jacket_fr_r_Aim
  bone2Name=spine_02
  swing1LimitDegrees=45
  swing1Motion=Free
  swing2LimitDegrees=45
  swing2Motion=Free
  twistLimitDegrees=45
  twistMotion=Free
 }
 {
  bone1Name=dyn_jacket_fr_r_MSTR
  bone2Name=dyn_jacket_fr_r_Aim
  swing1LimitDegrees=45
  swing1Motion=Free
  swing2LimitDegrees=45
  swing2Motion=Free
  twistLimitDegrees=45
  twistMotion=Free
 }
 {bone1Name=dyn_jacket_fr_r_01, bone2Name=dyn_jacket_fr_r_MSTR, swing1LimitDegrees=15, swing2LimitDegrees=15, twistLimitDegrees=15}
 {bone1Name=dyn_jacket_fr_r_02, bone2Name=dyn_jacket_fr_r_01, swing1LimitDegrees=20, swing2LimitDegrees=30, twistLimitDegrees=20}
 {bone1Name=dyn_jacket_fr_r_03, bone2Name=dyn_jacket_fr_r_02, swing1LimitDegrees=20, swing2LimitDegrees=30, twistLimitDegrees=20}
 {bone1Name=dyn_jacket_fr_r_04, bone2Name=dyn_jacket_fr_r_03, swing1LimitDegrees=20, swing2LimitDegrees=30, twistLimitDegrees=20}
 {
  bone1Name=dyn_Cloth_volume_l_B
  bone2Name=pelvis
  swing1LimitDegrees=45
  swing1Motion=Free
  swing2LimitDegrees=45
  swing2Motion=Free
  twistLimitDegrees=45
  twistMotion=Free
 }
 {
  bone1Name=dyn_Cloth_volume_l_A
  bone2Name=pelvis
  swing1LimitDegrees=45
  swing1Motion=Free
  swing2LimitDegrees=45
  swing2Motion=Free
  twistLimitDegrees=45
  twistMotion=Free
 }
 {
  bone1Name=dyn_Cloth_volume_r_A
  bone2Name=pelvis
  swing1LimitDegrees=45
  swing1Motion=Free
  swing2LimitDegrees=45
  swing2Motion=Free
  twistLimitDegrees=45
  twistMotion=Free
 }
 {
  bone1Name=dyn_Cloth_volume_r_B
  bone2Name=pelvis
  swing1LimitDegrees=45
  swing1Motion=Free
  swing2LimitDegrees=45
  swing2Motion=Free
  twistLimitDegrees=45
  twistMotion=Free
 }
 {bone1Name=dyn_jacket_r_03, bone2Name=dyn_jacket_r_02, swing1LimitDegrees=35, swing2LimitDegrees=20, twistLimitDegrees=5}
 {
  bone1Name=dyn_jacket_r_Aim
  bone2Name=spine_02
  swing1LimitDegrees=45
  swing1Motion=Free
  swing2LimitDegrees=45
  swing2Motion=Free
  twistLimitDegrees=45
  twistMotion=Free
 }
 {
  bone1Name=dyn_jacket_r_MSTR
  bone2Name=dyn_jacket_r_Aim
  swing1LimitDegrees=45
  swing1Motion=Free
  swing2LimitDegrees=45
  swing2Motion=Free
  twistLimitDegrees=45
  twistMotion=Free
 }
 {bone1Name=dyn_jacket_r_01, bone2Name=dyn_jacket_r_MSTR, swing1LimitDegrees=35, swing2LimitDegrees=20, twistLimitDegrees=5}
 {bone1Name=dyn_jacket_r_02, bone2Name=dyn_jacket_r_01, swing1LimitDegrees=35, swing2LimitDegrees=20, twistLimitDegrees=5}
 {bone1Name=dyn_jacket_l_03, bone2Name=dyn_jacket_l_02, swing1LimitDegrees=35, swing2LimitDegrees=45, twistLimitDegrees=15}
 {
  bone1Name=dyn_jacket_l_Aim
  bone2Name=spine_02
  swing1LimitDegrees=45
  swing1Motion=Free
  swing2LimitDegrees=45
  swing2Motion=Free
  twistLimitDegrees=45
  twistMotion=Free
 }
 {
  bone1Name=dyn_jacket_l_MSTR
  bone2Name=dyn_jacket_l_Aim
  swing1LimitDegrees=45
  swing1Motion=Free
  swing2LimitDegrees=45
  swing2Motion=Free
  twistLimitDegrees=45
  twistMotion=Free
 }
 {bone1Name=dyn_jacket_l_01, bone2Name=dyn_jacket_l_MSTR, swing1LimitDegrees=35, swing2LimitDegrees=45, twistLimitDegrees=15}
 {bone1Name=dyn_jacket_l_02, bone2Name=dyn_jacket_l_01, swing1LimitDegrees=35, swing2LimitDegrees=45, twistLimitDegrees=15}
 {
  bone1Name=dyn_Cloth_volume_l_C
  bone2Name=pelvis
  swing1LimitDegrees=45
  swing1Motion=Free
  swing2LimitDegrees=45
  swing2Motion=Free
  twistLimitDegrees=45
  twistMotion=Free
 }
 {
  bone1Name=dyn_Cloth_volume_r_C
  bone2Name=pelvis
  swing1LimitDegrees=45
  swing1Motion=Free
  swing2LimitDegrees=45
  swing2Motion=Free
  twistLimitDegrees=45
  twistMotion=Free
 }
 {
  bone1Name=dyn_jacket_fr_l_02
  bone2Name=dyn_jacket_l_02
  swing1LimitDegrees=45
  swing1Motion=Free
  swing2LimitDegrees=45
  swing2Motion=Free
  twistLimitDegrees=45
  twistMotion=Free
 }
]
