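# Physics asset for a medium female Pencil body with cuffed blazer and short skirt, defining capsule bodies and constraints for jacket secondary motion.
# Seven jacket chains descend from a master bone limited to spine_02, through free-swinging Aim/MSTR links into angle-limited segments.
# Free-rotation cross-links tie the side and back chains at their 02 segments.
# Pelvis, spine, thigh and eight pelvis-attached cloth-volume capsules supply body and skirt volume.
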PhysicsAsset /BRCosmetics/Characters/Player/Female/Medium/Bodies/F_MED_Pencil_B/BlazerCuff/Meshes/Pencil_BlazerCuff_MedB_Ubdy_ShortSkirt_Physics
bodies {
 dyn_Cloth_volume_bk [{
   shapeType=Capsule
   length=6.7346
   radius=9.3904
   shapeName=dyn_Cloth_volume_bk_capsule
   center=<-8.4718 4.2432 0>
   rotation=<rot 90 -3.5763 -13.576>
  }]
 dyn_Cloth_volume_fn [{
   shapeType=Capsule
   length=4.1135
   radius=9.0115
   shapeName=dyn_Cloth_volume_fn_capsule
   center=<-6.434 -3.953 0>
   rotation=<rot 90 180 174.9997>
  }]
 dyn_Cloth_volume_l_A [{
   shapeType=Capsule
   length=6.375
   radius=6.75
   shapeName=dyn_Cloth_volume_l_A_capsule
   center=<-5.0503 3.8372 -0.8889>
   rotation=<rot 85 10.0003 0>
  }]
 dyn_Cloth_volume_l_B [{shapeType=Capsule, length=4.9407, radius=8.1843, shapeName=dyn_Cloth_volume_l_B_capsule, center=<-7.541 4.0589 0>, rotation=<rot 90 180 175>}]
 dyn_Cloth_volume_l_C [{
   shapeType=Capsule
   length=4.9929
   radius=8.1321
   shapeName=dyn_Cloth_volume_l_C_capsule
   center=<-7.4857 4.319 -2.3444>
   rotation=<rot 85 -170.0003 179.9999>
  }]
 dyn_Cloth_volume_r_A [{
   shapeType=Capsule
   length=5.7415
   radius=8.2127
   shapeName=dyn_Cloth_volume_r_A_capsule
   center=<-8.372 2.3258 1.7398>
   rotation=<rot 85 -169.9998 -180>
  }]
 dyn_Cloth_volume_r_B [{
   shapeType=Capsule
   length=5.43
   radius=8.2337
   shapeName=dyn_Cloth_volume_r_B_capsule
   center=<-8.2093 4.153 0.8222>
   rotation=<rot 90 180 175>
  }]
 dyn_Cloth_volume_r_C [{
   shapeType=Capsule
   length=4.9929
   radius=8.1321
   shapeName=dyn_Cloth_volume_r_C_capsule
   center=<-7.3835 3.7394 1.184>
   rotation=<rot 85 9.9998 0>
  }]
 dyn_jacket_MSTR [{shapeType=Capsule, length=0.505, radius=0.505}]
 dyn_jacket_bk_01 [{shapeType=Capsule, length=10.5, radius=7, shapeName=dyn_jacket_bk_01_capsule, center=<4 -5 0>}]
 dyn_jacket_bk_02 [{shapeType=Capsule, length=10.5, radius=7, shapeName=dyn_jacket_bk_02_capsule, center=<4 -5 0>}]
 dyn_jacket_bk_03 [{shapeType=Capsule, length=10.5, radius=7, shapeName=dyn_jacket_bk_03_capsule, center=<4 -5 0>}]
 dyn_jacket_bk_Aim [{shapeType=Capsule, length=0.505, radius=0.505, shapeName=dyn_jacket_bk_Aim_capsule, rotation=<rot 0 -90 0>}]
 dyn_jacket_bk_MSTR [{shapeType=Capsule, length=0.505, radius=0.505, shapeName=dyn_jacket_bk_MSTR_capsule, rotation=<rot 0 -90 0>}]
 dyn_jacket_bk_l_01 [{shapeType=Capsule, length=9.625, radius=7, shapeName=dyn_jacket_bk_l_01_capsule, center=<4 0 5>, rotation=<rot 0 0 90>}]
 dyn_jacket_bk_l_02 [{shapeType=Capsule, length=9.625, radius=7, shapeName=dyn_jacket_bk_l_02_capsule, center=<4 0 5>, rotation=<rot 0 0 90>}]
 dyn_jacket_bk_l_03 [{shapeType=Capsule, length=9.375, radius=7.25, shapeName=dyn_jacket_bk_l_03_capsule, center=<4 0 5>, rotation=<rot -0.0001 0.0001 90>}]
 dyn_jacket_bk_l_Aim [{shapeType=Capsule, length=0.505, radius=0.505, shapeName=dyn_jacket_bk_l_Aim_capsule, rotation=<rot 0 -90 0>}]
 dyn_jacket_bk_l_MSTR [{shapeType=Capsule, length=0.505, radius=0.505, shapeName=dyn_jacket_bk_l_MSTR_capsule, rotation=<rot 0 -90 0>}]
 dyn_jacket_bk_r_01 [{shapeType=Capsule, length=9.625, radius=7, shapeName=dyn_jacket_bk_r_01_capsule, center=<4 -5 0>}]
 dyn_jacket_bk_r_02 [{shapeType=Capsule, length=9.625, radius=7, shapeName=dyn_jacket_bk_r_02_capsule, center=<4 -5 0>, rotation=<rot 0 0 5>}]
 dyn_jacket_bk_r_03 [{shapeType=Capsule, length=9.25, radius=7.375, shapeName=dyn_jacket_bk_r_03_capsule, center=<4 -5 0>, rotation=<rot 0 0 5>}]
 dyn_jacket_bk_r_Aim [{shapeType=Capsule, length=0.505, radius=0.505, shapeName=dyn_jacket_bk_r_Aim_capsule, rotation=<rot 0 -90 0>}]
 dyn_jacket_bk_r_MSTR [{shapeType=Capsule, length=0.505, radius=0.505, shapeName=dyn_jacket_bk_r_MSTR_capsule, rotation=<rot 0 -90 0>}]
 dyn_jacket_fr_l_01 [{shapeType=Capsule, length=6.75, radius=5.25, shapeName=dyn_jacket_fr_l_01_capsule, center=<4 -6 5>, rotation=<rot 0.0002 0.0002 90>}]
 dyn_jacket_fr_l_02 [{shapeType=Capsule, length=9.25, radius=4.75, shapeName=dyn_jacket_fr_l_02_capsule, center=<4 -6 5>, rotation=<rot 0.0001 0.0001 80>}]
 dyn_jacket_fr_l_03 [{shapeType=Capsule, length=15, radius=7.875, shapeName=dyn_jacket_fr_l_03_capsule, center=<4 -8 5>, rotation=<rot 0.0001 0 80>}]
 dyn_jacket_fr_l_04 [{shapeType=Capsule, length=11, radius=7.875, shapeName=dyn_jacket_fr_l_04_capsule, center=<4 -8 5>, rotation=<rot 0.0001 0 80>}]
 dyn_jacket_fr_l_Aim [{shapeType=Capsule, length=0.505, radius=0.505, shapeName=dyn_jacket_fr_l_Aim_capsule, rotation=<rot 0 -90 0>}]
 dyn_jacket_fr_l_MSTR [{shapeType=Capsule, length=0.505, radius=0.505, shapeName=dyn_jacket_fr_l_MSTR_capsule, rotation=<rot 0 -90 0>}]
 dyn_jacket_fr_r_01 [{shapeType=Capsule, length=5.375, radius=6.625, shapeName=dyn_jacket_fr_r_01_capsule, center=<4 7 4>, rotation=<rot 0.0001 -0.0001 115>}]
 dyn_jacket_fr_r_02 [{shapeType=Capsule, length=6.75, radius=5.25, shapeName=dyn_jacket_fr_r_02_capsule, center=<4 6 5>, rotation=<rot 0 0 90>}]
 dyn_jacket_fr_r_03 [{shapeType=Capsule, length=7.0808, radius=6.9192, shapeName=dyn_jacket_fr_r_03_capsule, center=<4 6 5>, rotation=<rot 0.0001 -0.0001 95>}]
 dyn_jacket_fr_r_04 [{shapeType=Capsule, length=7.0808, radius=6.9192, shapeName=dyn_jacket_fr_r_04_capsule, center=<4 6 5>, rotation=<rot 0.0001 -0.0001 95>}]
 dyn_jacket_fr_r_Aim [{shapeType=Capsule, length=0.505, radius=0.505, shapeName=dyn_jacket_fr_r_Aim_capsule, rotation=<rot 0 -90 0>}]
 dyn_jacket_fr_r_MSTR [{shapeType=Capsule, length=0.505, radius=0.505, shapeName=dyn_jacket_fr_r_MSTR_capsule, rotation=<rot 0 -90 0>}]
 dyn_jacket_l_01 [{shapeType=Capsule, length=5.375, radius=6.625, shapeName=dyn_jacket_l_01_capsule, center=<4 0 5>, rotation=<rot 0 0 90>}]
 dyn_jacket_l_02 [{shapeType=Capsule, length=12.75, radius=6.625, shapeName=dyn_jacket_l_02_capsule, center=<4 -4 5>, rotation=<rot 0 -9.9999 90>}]
 dyn_jacket_l_03 [{shapeType=Capsule, length=11, radius=6.75, shapeName=dyn_jacket_l_03_capsule, center=<4 -4 5>, rotation=<rot 0 -24.9998 89.9999>}]
 dyn_jacket_l_Aim [{shapeType=Capsule, length=0.505, radius=0.505, shapeName=dyn_jacket_l_Aim_capsule, rotation=<rot 0 -90 0>}]
 dyn_jacket_l_MSTR [{shapeType=Capsule, length=0.505, radius=0.505, shapeName=dyn_jacket_l_MSTR_capsule, rotation=<rot 0 -90 0>}]
 dyn_jacket_r_01 [{shapeType=Capsule, length=5.375, radius=6.625, shapeName=dyn_jacket_r_01_capsule, center=<5 -5 0>}]
 dyn_jacket_r_02 [{shapeType=Capsule, length=5.375, radius=6.625, shapeName=dyn_jacket_r_02_capsule, center=<5 -5 0>, rotation=<rot 9.9999 0 0>}]
 dyn_jacket_r_03 [{shapeType=Capsule, length=11.828, radius=7.297, shapeName=dyn_jacket_r_03_capsule, center=<5 -5 0>, rotation=<rot 24.9998 0 0.0001>}]
 dyn_jacket_r_Aim [{shapeType=Capsule, length=0.505, radius=0.505, shapeName=dyn_jacket_r_Aim_capsule, rotation=<rot 0 -90 0>}]
 dyn_jacket_r_MSTR [{shapeType=Capsule, length=0.505, radius=0.505, shapeName=dyn_jacket_r_MSTR_capsule, rotation=<rot 0 -90 0>}]
 pelvis [
  {shapeType=Capsule, length=6.7125, radius=11.8051, center=<-0.6285 -1.8567 0>}
  {shapeType=Capsule, length=9.415, radius=11.0952, shapeName=pelvis_capsule, center=<-3.9978 -3.5897 0>}
 ]
 spine_01 [{shapeType=Capsule, length=5.1863, radius=9.4291, center=<0 -0.1396 -0.0001>}]
 spine_02 [{shapeType=Capsule, length=3.5955, radius=9.005, center=<0 0.6676 -0.0001>}]
 thigh_l [{shapeType=Capsule, length=26.25, radius=11.25, center=<-21.6156 -1.1701 1.3426>, rotation=<rot 85 179.8042 179.9999>}]
 thigh_r [{shapeType=Capsule, length=27, radius=10.5, center=<21.0533 1.5 -1.1849>, rotation=<rot 80 179.9999 180>}]
}
constraints [
 {bone1Name=dyn_jacket_MSTR, bone2Name=spine_02, swing1LimitDegrees=45, swing2LimitDegrees=45, twistLimitDegrees=45}
 {
  bone1Name=spine_01
  bone2Name=pelvis
  swing1LimitDegrees=45
  swing1Motion=Free
  swing2LimitDegrees=45
  swing2Motion=Free
  twistLimitDegrees=45
  twistMotion=Free
 }
 {
  bone1Name=spine_02
  bone2Name=spine_01
  swing1LimitDegrees=45
  swing1Motion=Free
  swing2LimitDegrees=45
  swing2Motion=Free
  twistLimitDegrees=45
  twistMotion=Free
 }
 {
  bone1Name=thigh_l
  bone2Name=pelvis
  swing1LimitDegrees=45
  swing1Motion=Free
  swing2LimitDegrees=45
  swing2Motion=Free
  twistLimitDegrees=45
  twistMotion=Free
 }
 {
  bone1Name=thigh_r
  bone2Name=pelvis
  swing1LimitDegrees=45
  swing1Motion=Free
  swing2LimitDegrees=45
  swing2Motion=Free
  twistLimitDegrees=45
  twistMotion=Free
 }
 {bone1Name=dyn_jacket_r_03, bone2Name=dyn_jacket_r_02, swing1LimitDegrees=45, swing2LimitDegrees=35, twistLimitDegrees=15}
 {
  bone1Name=dyn_jacket_bk_Aim
  bone2Name=dyn_jacket_MSTR
  swing1LimitDegrees=45
  swing1Motion=Free
  swing2LimitDegrees=45
  swing2Motion=Free
  twistLimitDegrees=45
  twistMotion=Free
 }
 {
  bone1Name=dyn_jacket_bk_MSTR
  bone2Name=dyn_jacket_bk_Aim
  swing1LimitDegrees=45
  swing1Motion=Free
  swing2LimitDegrees=45
  swing2Motion=Free
  twistLimitDegrees=45
  twistMotion=Free
 }
 {bone1Name=dyn_jacket_bk_01, bone2Name=dyn_jacket_bk_MSTR, swing1LimitDegrees=45, swing2LimitDegrees=45, twistLimitDegrees=15}
 {bone1Name=dyn_jacket_bk_02, bone2Name=dyn_jacket_bk_01, swing1LimitDegrees=45, swing2LimitDegrees=45, twistLimitDegrees=15}
 {bone1Name=dyn_jacket_bk_03, bone2Name=dyn_jacket_bk_02, swing1LimitDegrees=45, swing2LimitDegrees=45, twistLimitDegrees=15}
 {
  bone1Name=dyn_jacket_bk_l_Aim
  bone2Name=dyn_jacket_MSTR
  swing1LimitDegrees=45
  swing1Motion=Free
  swing2LimitDegrees=45
  swing2Motion=Free
  twistLimitDegrees=45
  twistMotion=Free
 }
 {
  bone1Name=dyn_jacket_bk_l_MSTR
  bone2Name=dyn_jacket_bk_l_Aim
  swing1LimitDegrees=45
  swing1Motion=Free
  swing2LimitDegrees=45
  swing2Motion=Free
  twistLimitDegrees=45
  twistMotion=Free
 }
 {bone1Name=dyn_jacket_bk_l_01, bone2Name=dyn_jacket_bk_l_MSTR, swing1LimitDegrees=35, swing2LimitDegrees=45, twistLimitDegrees=15}
 {bone1Name=dyn_jacket_bk_l_02, bone2Name=dyn_jacket_bk_l_01, swing1LimitDegrees=35, swing2LimitDegrees=45, twistLimitDegrees=15}
 {bone1Name=dyn_jacket_bk_l_03, bone2Name=dyn_jacket_bk_l_02, swing1LimitDegrees=35, swing2LimitDegrees=45, twistLimitDegrees=15}
 {
  bone1Name=dyn_jacket_bk_r_Aim
  bone2Name=dyn_jacket_MSTR
  swing1LimitDegrees=45
  swing1Motion=Free
  swing2LimitDegrees=45
  swing2Motion=Free
  twistLimitDegrees=45
  twistMotion=Free
 }
 {
  bone1Name=dyn_jacket_bk_r_MSTR
  bone2Name=dyn_jacket_bk_r_Aim
  swing1LimitDegrees=45
  swing1Motion=Free
  swing2LimitDegrees=45
  swing2Motion=Free
  twistLimitDegrees=45
  twistMotion=Free
 }
 {bone1Name=dyn_jacket_bk_r_01, bone2Name=dyn_jacket_bk_r_MSTR, swing1LimitDegrees=45, swing2LimitDegrees=35, twistLimitDegrees=15}
 {bone1Name=dyn_jacket_bk_r_02, bone2Name=dyn_jacket_bk_r_01, swing1LimitDegrees=45, swing2LimitDegrees=35, twistLimitDegrees=15}
 {bone1Name=dyn_jacket_bk_r_03, bone2Name=dyn_jacket_bk_r_02, swing1LimitDegrees=45, swing2LimitDegrees=35, twistLimitDegrees=15}
 {
  bone1Name=dyn_jacket_fr_l_Aim
  bone2Name=dyn_jacket_MSTR
  swing1LimitDegrees=45
  swing1Motion=Free
  swing2LimitDegrees=45
  swing2Motion=Free
  twistLimitDegrees=45
  twistMotion=Free
 }
 {
  bone1Name=dyn_jacket_fr_l_MSTR
  bone2Name=dyn_jacket_fr_l_Aim
  swing1LimitDegrees=45
  swing1Motion=Free
  swing2LimitDegrees=45
  swing2Motion=Free
  twistLimitDegrees=45
  twistMotion=Free
 }
 {bone1Name=dyn_jacket_fr_l_01, bone2Name=dyn_jacket_fr_l_MSTR, swing1LimitDegrees=15, swing2LimitDegrees=15, twistLimitDegrees=35}
 {bone1Name=dyn_jacket_fr_l_02, bone2Name=dyn_jacket_fr_l_01, swing1LimitDegrees=30, swing2LimitDegrees=40, twistLimitDegrees=20}
 {bone1Name=dyn_jacket_fr_l_03, bone2Name=dyn_jacket_fr_l_02, swing1LimitDegrees=30, swing2LimitDegrees=40, twistLimitDegrees=20}
 {bone1Name=dyn_jacket_fr_l_04, bone2Name=dyn_jacket_fr_l_03, swing1LimitDegrees=30, swing2LimitDegrees=40, twistLimitDegrees=20}
 {
  bone1Name=dyn_jacket_fr_r_Aim
  bone2Name=dyn_jacket_MSTR
  swing1LimitDegrees=45
  swing1Motion=Free
  swing2LimitDegrees=45
  swing2Motion=Free
  twistLimitDegrees=45
  twistMotion=Free
 }
 {
  bone1Name=dyn_jacket_fr_r_MSTR
  bone2Name=dyn_jacket_fr_r_Aim
  swing1LimitDegrees=45
  swing1Motion=Free
  swing2LimitDegrees=45
  swing2Motion=Free
  twistLimitDegrees=45
  twistMotion=Free
 }
 {bone1Name=dyn_jacket_fr_r_01, bone2Name=dyn_jacket_fr_r_MSTR, swing1LimitDegrees=15, swing2LimitDegrees=15, twistLimitDegrees=35}
 {bone1Name=dyn_jacket_fr_r_02, bone2Name=dyn_jacket_fr_r_01, swing1LimitDegrees=35, swing2LimitDegrees=45, twistLimitDegrees=15}
 {bone1Name=dyn_jacket_fr_r_03, bone2Name=dyn_jacket_fr_r_02, swing1LimitDegrees=35, swing2LimitDegrees=45, twistLimitDegrees=15}
 {bone1Name=dyn_jacket_fr_r_04, bone2Name=dyn_jacket_fr_r_03, swing1LimitDegrees=35, swing2LimitDegrees=45, twistLimitDegrees=15}
 {
  bone1Name=dyn_jacket_l_Aim
  bone2Name=dyn_jacket_MSTR
  swing1LimitDegrees=45
  swing1Motion=Free
  swing2LimitDegrees=45
  swing2Motion=Free
  twistLimitDegrees=45
  twistMotion=Free
 }
 {
  bone1Name=dyn_jacket_l_MSTR
  bone2Name=dyn_jacket_l_Aim
  swing1LimitDegrees=45
  swing1Motion=Free
  swing2LimitDegrees=45
  swing2Motion=Free
  twistLimitDegrees=45
  twistMotion=Free
 }
 {bone1Name=dyn_jacket_l_01, bone2Name=dyn_jacket_l_MSTR, swing1LimitDegrees=35, swing2LimitDegrees=45, twistLimitDegrees=15}
 {bone1Name=dyn_jacket_l_02, bone2Name=dyn_jacket_l_01, swing1LimitDegrees=35, swing2LimitDegrees=45, twistLimitDegrees=15}
 {bone1Name=dyn_jacket_l_03, bone2Name=dyn_jacket_l_02, swing1LimitDegrees=35, swing2LimitDegrees=45, twistLimitDegrees=15}
 {
  bone1Name=dyn_jacket_r_Aim
  bone2Name=dyn_jacket_MSTR
  swing1LimitDegrees=45
  swing1Motion=Free
  swing2LimitDegrees=45
  swing2Motion=Free
  twistLimitDegrees=45
  twistMotion=Free
 }
 {
  bone1Name=dyn_jacket_r_MSTR
  bone2Name=dyn_jacket_r_Aim
  swing1LimitDegrees=45
  swing1Motion=Free
  swing2LimitDegrees=45
  swing2Motion=Free
  twistLimitDegrees=45
  twistMotion=Free
 }
 {bone1Name=dyn_jacket_r_01, bone2Name=dyn_jacket_r_MSTR, swing1LimitDegrees=45, swing2LimitDegrees=35, twistLimitDegrees=15}
 {bone1Name=dyn_jacket_r_02, bone2Name=dyn_jacket_r_01, swing1LimitDegrees=45, swing2LimitDegrees=35, twistLimitDegrees=15}
 {
  bone1Name=dyn_jacket_bk_l_02
  bone2Name=dyn_jacket_bk_02
  swing1LimitDegrees=45
  swing1Motion=Free
  swing2LimitDegrees=45
  swing2Motion=Free
  twistLimitDegrees=45
  twistMotion=Free
 }
 {
  bone1Name=dyn_jacket_bk_r_02
  bone2Name=dyn_jacket_bk_02
  swing1LimitDegrees=45
  swing1Motion=Free
  swing2LimitDegrees=45
  swing2Motion=Free
  twistLimitDegrees=45
  twistMotion=Free
 }
 {
  bone1Name=dyn_jacket_r_02
  bone2Name=dyn_jacket_bk_r_02
  swing1LimitDegrees=45
  swing1Motion=Free
  swing2LimitDegrees=45
  swing2Motion=Free
  twistLimitDegrees=45
  twistMotion=Free
 }
 {
  bone1Name=dyn_jacket_l_02
  bone2Name=dyn_jacket_bk_l_02
  swing1LimitDegrees=45
  swing1Motion=Free
  swing2LimitDegrees=45
  swing2Motion=Free
  twistLimitDegrees=45
  twistMotion=Free
 }
 {
  bone1Name=dyn_Cloth_volume_bk
  bone2Name=pelvis
  swing1LimitDegrees=45
  swing1Motion=Free
  swing2LimitDegrees=45
  swing2Motion=Free
  twistLimitDegrees=45
  twistMotion=Free
 }
 {
  bone1Name=dyn_Cloth_volume_fn
  bone2Name=pelvis
  swing1LimitDegrees=45
  swing1Motion=Free
  swing2LimitDegrees=45
  swing2Motion=Free
  twistLimitDegrees=45
  twistMotion=Free
 }
 {
  bone1Name=dyn_Cloth_volume_l_C
  bone2Name=pelvis
  swing1LimitDegrees=45
  swing1Motion=Free
  swing2LimitDegrees=45
  swing2Motion=Free
  twistLimitDegrees=45
  twistMotion=Free
 }
 {
  bone1Name=dyn_Cloth_volume_l_A
  bone2Name=pelvis
  swing1LimitDegrees=45
  swing1Motion=Free
  swing2LimitDegrees=45
  swing2Motion=Free
  twistLimitDegrees=45
  twistMotion=Free
 }
 {
  bone1Name=dyn_Cloth_volume_l_B
  bone2Name=pelvis
  swing1LimitDegrees=45
  swing1Motion=Free
  swing2LimitDegrees=45
  swing2Motion=Free
  twistLimitDegrees=45
  twistMotion=Free
 }
 {
  bone1Name=dyn_Cloth_volume_r_A
  bone2Name=pelvis
  swing1LimitDegrees=45
  swing1Motion=Free
  swing2LimitDegrees=45
  swing2Motion=Free
  twistLimitDegrees=45
  twistMotion=Free
 }
 {
  bone1Name=dyn_Cloth_volume_r_B
  bone2Name=pelvis
  swing1LimitDegrees=45
  swing1Motion=Free
  swing2LimitDegrees=45
  swing2Motion=Free
  twistLimitDegrees=45
  twistMotion=Free
 }
 {
  bone1Name=dyn_Cloth_volume_r_C
  bone2Name=pelvis
  swing1LimitDegrees=45
  swing1Motion=Free
  swing2LimitDegrees=45
  swing2Motion=Free
  twistLimitDegrees=45
  twistMotion=Free
 }
]
