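# Post-process AnimBlueprint for the female medium Pencil BlazerCuff cosmetic body, driving jacket and clothing secondary motion.
# Copies the parent mesh pose through a deform Control Rig and weighted bone constraints for jacket and cloth-volume helpers.
# Skirt short and Skirt Long bools select between RigidBody branches fed by cached 'presim' poses.
# Additive bone offsets handle sliding, front end and spine rotation; skydiving blends in constraints, AnimDynamics and alternate simulation settings.
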AnimBlueprint /BRCosmetics/Characters/Player/Female/Medium/Bodies/F_MED_Pencil_B/BlazerCuff/Meshes/Pencil_BlazerCuff_MedB_Ubdy_AnimBP
target_skeleton=&/Pencil_BlazerCuff_MedB_Ubdy_Skeleton
graphs [
 AnimGraph {nodes [
   "Constraint - dyn_shirt_arm_fix (P)"(Alpha=1, ConstraintWeights_0=1) {
    alphaBoolBlend {blendOption=Linear}
    boneToModify {boneName=dyn_shirt_arm_fix}
    constraintSetup [{targetBone {boneName=clavicle_r}}]
    constraintWeights [1]
   }
   Input Pose()
   n1 = Get bIsSkyDiving()
   RigidBody!(ComponentPose, bAlphaBoolEnabled:bool=@n1, OverrideWorldGravity:Vector="0,0,-980", ExternalForce) -> (Pose) {
    alphaInputType=Bool
    bClampLinearTranslationLimitToRefPose=true
    bOverrideWorldGravity=true
    overridePhysicsAsset=&/Pencil_BlazerCuff_MedB_Ubdy_ShortSkirt_Physics
    simulationSpace=BaseBoneSpace
    alphaBoolBlend {blendInTime=0.1, blendOption=Linear, blendOutTime=0.1}
    baseBoneRef {boneName=pelvis}
    componentAppliedLinearAccClamp=<3000 3000 3000>
    componentLinearAccScale=<0.3 0.3 0.4>
    componentLinearVelScale=<0.3 0.3 0.4>
    simSpaceSettings {
     maxAngularAcceleration=1000
     maxAngularVelocity=3
     maxLinearAcceleration=700
     maxLinearVelocity=500
     velocityScaleZ=0
     externalLinearDragV=<0.2 0.2>
    }
   }
   n2 = Get bIsSliding()
   n3 = "Transform (Modify) Bone - Bone: dyn_jacket_bk_MSTR"!(
    ComponentPose=Local To Component(LocalPose=Use cached pose 'presim'())
    bAlphaBoolEnabled:bool=@n2
    Translation="0,-2,0"
    Rotation="0,-50,0"
    Scale="1,1,1"
   ) -> (Pose) {
    alphaInputType=Bool
    rotationMode=BMM_Additive
    rotationSpace=BCS_BoneSpace
    translationMode=BMM_Additive
    alphaBoolBlend {blendInTime=0.1, blendOption=Linear, blendOutTime=0.3}
    alphaScaleBiasClamp {bInterpResult=true, interpSpeedIncreasing=2}
    boneToModify {boneName=dyn_jacket_bk_MSTR}
    rotation {yaw=-50}
    translation {y=-2}
   }
   n4 = "Transform (Modify) Bone - Bone: dyn_jacket_bk_l_MSTR"!(ComponentPose=@n3, bAlphaBoolEnabled:bool=@n2, Translation="0,0,2", Rotation="40,-25,0", Scale="1,1,1") -> (Pose) {
    alphaInputType=Bool
    rotationMode=BMM_Additive
    rotationSpace=BCS_BoneSpace
    translationMode=BMM_Additive
    alphaBoolBlend {blendInTime=0.1, blendOption=Linear, blendOutTime=0.3}
    alphaScaleBiasClamp {bInterpResult=true, interpSpeedIncreasing=2}
    boneToModify {boneName=dyn_jacket_bk_l_MSTR}
    rotation {pitch=40, yaw=-25}
    translation {z=2}
   }
   n5 = "Transform (Modify) Bone - Bone: dyn_jacket_bk_r_MSTR"!(ComponentPose=@n4, bAlphaBoolEnabled:bool=@n2, Translation="0,-2,0", Rotation="35,-50,0", Scale="1,1,1") -> (Pose) {
    alphaInputType=Bool
    rotationMode=BMM_Additive
    rotationSpace=BCS_BoneSpace
    translationMode=BMM_Additive
    alphaBoolBlend {blendInTime=0.1, blendOption=Linear, blendOutTime=0.3}
    alphaScaleBiasClamp {bInterpResult=true, interpSpeedIncreasing=2}
    boneToModify {boneName=dyn_jacket_bk_r_MSTR}
    rotation {pitch=35, yaw=-50}
    translation {y=-2}
   }
   n6 = RigidBody(ComponentPose=@n5, OverrideWorldGravity=Get GravityOverride()) {
    alphaInputType=Bool
    bClampLinearTranslationLimitToRefPose=true
    bOverrideWorldGravity=true
    overridePhysicsAsset=&/Pencil_BlazerCuff_MedB_Ubdy_Perf_Physics
    simulationSpace=BaseBoneSpace
    alphaBoolBlend {blendInTime=0.1, blendOption=Linear, blendOutTime=0.1}
    baseBoneRef {boneName=pelvis}
    componentAppliedLinearAccClamp=<3000 3000 3000>
    componentLinearAccScale=<0.3 0.3 0.5>
    componentLinearVelScale=<0.3 0.3 0.5>
    simSpaceSettings {
     maxAngularAcceleration=1000
     maxAngularVelocity=3
     maxLinearAcceleration=700
     maxLinearVelocity=500
     velocityScaleZ=0
     worldAlpha=1
     externalLinearDragV=<0.2 0.2>
    }
   }
   n7 = Get bIsSliding()
   n8 = "Transform (Modify) Bone - Bone: dyn_jacket_r_MSTR"!(ComponentPose=@n6, bAlphaBoolEnabled:bool=@n7, Translation="0,-2,2", Rotation="20,-55,10", Scale="1,1,1") -> (Pose) {
    alphaInputType=Bool
    rotationMode=BMM_Additive
    rotationSpace=BCS_BoneSpace
    translationMode=BMM_Additive
    alphaBoolBlend {blendInTime=0.1, blendOption=Linear, blendOutTime=0.3}
    alphaScaleBiasClamp {bInterpResult=true, interpSpeedIncreasing=2}
    boneToModify {boneName=dyn_jacket_r_MSTR}
    rotation=<rot 20 -55 10>
    translation {y=-2, z=2}
   }
   n9 = "Transform (Modify) Bone - Bone: dyn_jacket_fr_r_MSTR"!(ComponentPose=@n8, bAlphaBoolEnabled:bool=@n7, Translation="0,-2,-2", Rotation="60,60,40", Scale="1,1,1") -> (Pose) {
    alphaInputType=Bool
    rotationMode=BMM_Additive
    rotationSpace=BCS_BoneSpace
    translationMode=BMM_Additive
    alphaBoolBlend {blendInTime=0.1, blendOption=Linear, blendOutTime=0.3}
    alphaScaleBiasClamp {bInterpResult=true, interpSpeedIncreasing=2}
    boneToModify {boneName=dyn_jacket_fr_r_MSTR}
    rotation=<rot 60 60 40>
    translation {y=-2, z=-2}
   }
   n10 = "Transform (Modify) Bone - Bone: dyn_jacket_l_MSTR"!(ComponentPose=@n9, bAlphaBoolEnabled:bool=@n7, Translation="0,-2,2", Rotation="20,-50,10", Scale="1,1,1") -> (Pose) {
    alphaInputType=Bool
    rotationMode=BMM_Additive
    rotationSpace=BCS_BoneSpace
    translationMode=BMM_Additive
    alphaBoolBlend {blendInTime=0.1, blendOption=Linear, blendOutTime=0.3}
    alphaScaleBiasClamp {bInterpResult=true, interpSpeedIncreasing=2}
    boneToModify {boneName=dyn_jacket_l_MSTR}
    rotation=<rot 20 -50 10>
    translation {y=-2, z=2}
   }
   n11 = "Transform (Modify) Bone - Bone: dyn_jacket_fr_l_MSTR"!(ComponentPose=@n10, bAlphaBoolEnabled:bool=@n7, Translation="0,-2,2", Rotation="0,-50,0.001", Scale="1,1,1") -> (Pose) {
    alphaInputType=Bool
    rotationMode=BMM_Additive
    rotationSpace=BCS_BoneSpace
    translationMode=BMM_Additive
    alphaBoolBlend {blendInTime=0.1, blendOption=Linear, blendOutTime=0.3}
    alphaScaleBiasClamp {bInterpResult=true, interpSpeedIncreasing=2}
    boneToModify {boneName=dyn_jacket_fr_l_MSTR}
    rotation {roll=0.001, yaw=-50}
    translation {y=-2, z=2}
   }
   n12 = "Transform (Modify) Bone - Bone: dyn_jacket_bk_MSTR"(ComponentPose=@n11, bAlphaBoolEnabled=@n7, Translation="0,-2,0", Rotation="0,-15,0", Scale="1,1,1") {
    alphaInputType=Bool
    rotationMode=BMM_Additive
    rotationSpace=BCS_BoneSpace
    translationMode=BMM_Additive
    alphaBoolBlend {blendInTime=0.1, blendOption=Linear, blendOutTime=0.3}
    alphaScaleBiasClamp {bInterpResult=true, interpSpeedIncreasing=2}
    boneToModify {boneName=dyn_jacket_bk_MSTR}
    rotation {yaw=-15}
    translation {y=-2}
   }
   n13 = "Transform (Modify) Bone - Bone: dyn_jacket_bk_l_MSTR"(ComponentPose=@n12, bAlphaBoolEnabled=@n7, Translation="0,0,2", Rotation="15,0,0", Scale="1,1,1") {
    alphaInputType=Bool
    rotationMode=BMM_Additive
    rotationSpace=BCS_BoneSpace
    translationMode=BMM_Additive
    alphaBoolBlend {blendInTime=0.1, blendOption=Linear, blendOutTime=0.3}
    alphaScaleBiasClamp {bInterpResult=true, interpSpeedIncreasing=2}
    boneToModify {boneName=dyn_jacket_bk_l_MSTR}
    rotation {pitch=15}
    translation {z=2}
   }
   n14 = "Transform (Modify) Bone - Bone: dyn_jacket_bk_r_MSTR"(ComponentPose=@n13, bAlphaBoolEnabled=@n7, Translation="0,-2,0", Rotation="0,-15,0", Scale="1,1,1") {
    alphaInputType=Bool
    rotationMode=BMM_Additive
    rotationSpace=BCS_BoneSpace
    translationMode=BMM_Additive
    alphaBoolBlend {blendInTime=0.1, blendOption=Linear, blendOutTime=0.3}
    alphaScaleBiasClamp {bInterpResult=true, interpSpeedIncreasing=2}
    boneToModify {boneName=dyn_jacket_bk_r_MSTR}
    rotation {yaw=-15}
    translation {y=-2}
   }
   n15 = Get skydive() -> (skydive:double)
   n16 = "Constraint - dyn_jacket_bk_l_01 (P)"(ComponentPose=@n14, Alpha=@n15, ConstraintWeights_0=1) {
    alphaBoolBlend {blendOption=Linear}
    boneToModify {boneName=dyn_jacket_bk_l_01}
    constraintSetup [{targetBone {boneName=dyn_jacket_l_01}}]
    constraintWeights [1]
   }
   n17 = "Constraint - dyn_jacket_bk_l_02 (P)"(ComponentPose=@n16, Alpha=@n15, ConstraintWeights_0=1) {
    alphaBoolBlend {blendOption=Linear}
    boneToModify {boneName=dyn_jacket_bk_l_02}
    constraintSetup [{targetBone {boneName=dyn_jacket_l_02}}]
    constraintWeights [1]
   }
   n18 = "Constraint - dyn_jacket_bk_l_03 (P)"(ComponentPose=@n17, Alpha=@n15, ConstraintWeights_0=1) {
    alphaBoolBlend {blendOption=Linear}
    boneToModify {boneName=dyn_jacket_bk_l_03}
    constraintSetup [{targetBone {boneName=dyn_jacket_l_03}}]
    constraintWeights [1]
   }
   n19 = "Constraint - dyn_jacket_bk_r_01 (P)"(ComponentPose=@n18, Alpha=@n15, ConstraintWeights_0=1) {
    alphaBoolBlend {blendOption=Linear}
    boneToModify {boneName=dyn_jacket_bk_r_01}
    constraintSetup [{targetBone {boneName=dyn_jacket_r_01}}]
    constraintWeights [1]
   }
   n20 = "Constraint - dyn_jacket_bk_r_02 (P)"(ComponentPose=@n19, Alpha=@n15, ConstraintWeights_0=1) {
    alphaBoolBlend {blendOption=Linear}
    boneToModify {boneName=dyn_jacket_bk_r_02}
    constraintSetup [{targetBone {boneName=dyn_jacket_r_02}}]
    constraintWeights [1]
   }
   n21 = "Constraint - dyn_jacket_bk_r_03 (P)"(ComponentPose=@n20, Alpha=@n15, ConstraintWeights_0=1) {
    alphaBoolBlend {blendOption=Linear}
    boneToModify {boneName=dyn_jacket_bk_r_03}
    constraintSetup [{targetBone {boneName=dyn_jacket_r_03}}]
    constraintWeights [1]
   }
   n22 = "Constraint - dyn_jacket_bk_01 (PP)"!(ComponentPose=@n21, Alpha=@n15, ConstraintWeights_0=0.5, ConstraintWeights_1:float=0.5) -> (Pose) {
    alphaBoolBlend {blendOption=Linear}
    boneToModify {boneName=dyn_jacket_bk_01}
    constraintSetup [{targetBone {boneName=dyn_jacket_bk_r_01}}, {targetBone {boneName=dyn_jacket_bk_l_01}}]
    constraintWeights [1,1]
   }
   n23 = "Constraint - dyn_jacket_bk_02 (PP)"!(ComponentPose=@n22, Alpha=@n15, ConstraintWeights_0=0.5, ConstraintWeights_1:float=0.5) -> (Pose) {
    alphaBoolBlend {blendOption=Linear}
    boneToModify {boneName=dyn_jacket_bk_02}
    constraintSetup [{targetBone {boneName=dyn_jacket_bk_r_02}}, {targetBone {boneName=dyn_jacket_bk_l_02}}]
    constraintWeights [1,1]
   }
   n24 = "Constraint - dyn_jacket_bk_03 (PP)"!(ComponentPose=@n23, Alpha=@n15, ConstraintWeights_0=0.5, ConstraintWeights_1:float=0.5) -> (Pose) {
    alphaBoolBlend {blendOption=Linear}
    boneToModify {boneName=dyn_jacket_bk_03}
    constraintSetup [{targetBone {boneName=dyn_jacket_bk_r_03}}, {targetBone {boneName=dyn_jacket_bk_l_03}}]
    constraintWeights [1,1]
   }
   n25 = Get skydive_anim dynamics() -> (skydive_anim dynamics:double)
   n26 = Anim Dynamics - Bone: dyn_jacket_bk_01(ComponentPose=@n24, Alpha=@n25, ExternalForce="0,0,0") {
    angularDampingOverride=0.99
    angularSpringConstant=50
    bAngularSpring=true
    bGravityOverrideInSimSpace=true
    bOverrideAngularDamping=true
    simulationSpace=RootRelative
    alphaBoolBlend {blendOption=Linear}
    boundBone {boneName=dyn_jacket_bk_01}
    componentAppliedLinearAccClamp=<3500 3500 3500>
    componentLinearAccScale=<1 1 1>
    componentLinearVelScale=<1 1 1>
    gravityOverride {z=-810}
    physicsBodyDefinitions [{
      sphereCollisionRadius=0
      boundBone {boneName=dyn_jacket_bk_01}
      boxExtents=<20 20 20>
      constraintSetup {angularLimitsMax=<5 5 0>, angularLimitsMin=<-5 -5 -45>, angularTarget=<0 0 0>}
      localJointOffset=<2 0 0>
     }]
    relativeSpaceBone {boneName=root}
   }
   n27 = Anim Dynamics - Bone: dyn_jacket_bk_02(ComponentPose=@n26, Alpha=@n25, ExternalForce="0,0,0") {
    angularDampingOverride=0.99
    angularSpringConstant=50
    bAngularSpring=true
    bGravityOverrideInSimSpace=true
    bOverrideAngularDamping=true
    bUseSphericalLimits=true
    simulationSpace=RootRelative
    alphaBoolBlend {blendOption=Linear}
    boundBone {boneName=dyn_jacket_bk_02}
    componentAppliedLinearAccClamp=<3500 3500 3500>
    componentLinearAccScale=<1 1 1>
    componentLinearVelScale=<1 1 1>
    gravityOverride {z=-810}
    physicsBodyDefinitions [{
      sphereCollisionRadius=0
      boundBone {boneName=dyn_jacket_bk_02}
      boxExtents=<20 20 20>
      constraintSetup {angularLimitsMax=<5 5 0>, angularLimitsMin=<-5 -5 -45>, angularTarget=<0 0 0>}
      localJointOffset=<2 0 0>
     }]
    relativeSpaceBone {boneName=root}
    sphericalLimits [{limitRadius=15, drivingBone {boneName=pelvis}, sphereLocalOffset=<0 2 0>}]
   }
   n28 = Anim Dynamics - Bone: dyn_jacket_bk_03(ComponentPose=@n27, Alpha=@n25, ExternalForce="0,0,0") {
    angularDampingOverride=0.99
    angularSpringConstant=50
    bAngularSpring=true
    bGravityOverrideInSimSpace=true
    bOverrideAngularDamping=true
    bUseSphericalLimits=true
    simulationSpace=RootRelative
    alphaBoolBlend {blendOption=Linear}
    boundBone {boneName=dyn_jacket_bk_03}
    componentAppliedLinearAccClamp=<3500 3500 3500>
    componentLinearAccScale=<1 1 1>
    componentLinearVelScale=<1 1 1>
    gravityOverride {z=-810}
    physicsBodyDefinitions [{
      sphereCollisionRadius=0
      boundBone {boneName=dyn_jacket_bk_03}
      boxExtents=<20 20 20>
      constraintSetup {angularLimitsMax=<5 5 0>, angularLimitsMin=<-5 -5 -45>, angularTarget=<0 0 0>}
      localJointOffset=<2 0 0>
     }]
    relativeSpaceBone {boneName=root}
    sphericalLimits [{limitRadius=15, drivingBone {boneName=pelvis}, sphereLocalOffset=<0 1 0>}]
   }
   n29 = Anim Dynamics - Bone: dyn_jacket_bk_l_01(ComponentPose=@n28, Alpha=@n25, ExternalForce="0,0,0") {
    angularDampingOverride=0.99
    angularSpringConstant=50
    bAngularSpring=true
    bGravityOverrideInSimSpace=true
    bOverrideAngularDamping=true
    simulationSpace=RootRelative
    alphaBoolBlend {blendOption=Linear}
    boundBone {boneName=dyn_jacket_bk_l_01}
    componentAppliedLinearAccClamp=<3500 3500 3500>
    componentLinearAccScale=<1 1 1>
    componentLinearVelScale=<1 1 1>
    gravityOverride {z=-810}
    physicsBodyDefinitions [{
      sphereCollisionRadius=0
      boundBone {boneName=dyn_jacket_bk_l_01}
      boxExtents=<20 20 20>
      constraintSetup {angularLimitsMax=<5 0 5>, angularLimitsMin=<-5 -45 -5>, angularTarget=<0 0 0>}
      localJointOffset=<2 0 0>
     }]
    relativeSpaceBone {boneName=root}
   }
   n30 = Anim Dynamics - Bone: dyn_jacket_bk_l_02(ComponentPose=@n29, Alpha=@n25, ExternalForce="0,0,0") {
    angularDampingOverride=0.99
    angularSpringConstant=50
    bAngularSpring=true
    bGravityOverrideInSimSpace=true
    bOverrideAngularDamping=true
    bUseSphericalLimits=true
    simulationSpace=RootRelative
    alphaBoolBlend {blendOption=Linear}
    boundBone {boneName=dyn_jacket_bk_l_02}
    componentAppliedLinearAccClamp=<3500 3500 3500>
    componentLinearAccScale=<1 1 1>
    componentLinearVelScale=<1 1 1>
    gravityOverride {z=-810}
    physicsBodyDefinitions [{
      sphereCollisionRadius=0
      boundBone {boneName=dyn_jacket_bk_l_02}
      boxExtents=<20 20 20>
      constraintSetup {angularLimitsMax=<5 0 5>, angularLimitsMin=<-5 -45 -5>, angularTarget=<0 0 0>}
      localJointOffset=<2 0 0>
     }]
    relativeSpaceBone {boneName=root}
    sphericalLimits [
     {limitRadius=15, drivingBone {boneName=pelvis}, sphereLocalOffset=<0 1 -2>}
     {limitRadius=12, drivingBone {boneName=thigh_l}, sphereLocalOffset=<-5 0 0>}
    ]
   }
   n31 = Anim Dynamics - Bone: dyn_jacket_bk_l_03(ComponentPose=@n30, Alpha=@n25, ExternalForce="0,0,0") {
    angularDampingOverride=0.99
    angularSpringConstant=50
    bAngularSpring=true
    bGravityOverrideInSimSpace=true
    bOverrideAngularDamping=true
    bUseSphericalLimits=true
    simulationSpace=RootRelative
    alphaBoolBlend {blendOption=Linear}
    boundBone {boneName=dyn_jacket_bk_l_03}
    componentAppliedLinearAccClamp=<3500 3500 3500>
    componentLinearAccScale=<1 1 1>
    componentLinearVelScale=<1 1 1>
    gravityOverride {z=-810}
    physicsBodyDefinitions [{
      sphereCollisionRadius=0
      boundBone {boneName=dyn_jacket_bk_l_03}
      boxExtents=<20 20 20>
      constraintSetup {angularLimitsMax=<5 0 5>, angularLimitsMin=<-5 -45 -5>, angularTarget=<0 0 0>}
      localJointOffset=<2 0 0>
     }]
    relativeSpaceBone {boneName=root}
    sphericalLimits [
     {limitRadius=15.5, drivingBone {boneName=pelvis}, sphereLocalOffset=<0 1 -2>}
     {limitRadius=12, drivingBone {boneName=thigh_l}, sphereLocalOffset=<-5 0 0>}
    ]
   }
   n32 = Anim Dynamics - Bone: dyn_jacket_bk_r_01(ComponentPose=@n31, Alpha=@n25, ExternalForce="0,0,0") {
    angularDampingOverride=0.99
    angularSpringConstant=50
    bAngularSpring=true
    bGravityOverrideInSimSpace=true
    bOverrideAngularDamping=true
    simulationSpace=RootRelative
    alphaBoolBlend {blendOption=Linear}
    boundBone {boneName=dyn_jacket_bk_r_01}
    componentAppliedLinearAccClamp=<3500 3500 3500>
    componentLinearAccScale=<1 1 1>
    componentLinearVelScale=<1 1 1>
    gravityOverride {z=-810}
    physicsBodyDefinitions [{
      sphereCollisionRadius=0
      boundBone {boneName=dyn_jacket_bk_r_01}
      boxExtents=<20 20 20>
      constraintSetup {angularLimitsMax=<5 5 0>, angularLimitsMin=<-5 -5 -45>, angularTarget=<0 0 0>}
      localJointOffset=<2 0 0>
     }]
    relativeSpaceBone {boneName=root}
   }
   n33 = Anim Dynamics - Bone: dyn_jacket_bk_r_02(ComponentPose=@n32, Alpha=@n25, ExternalForce="0,0,0") {
    angularDampingOverride=0.99
    angularSpringConstant=50
    bAngularSpring=true
    bGravityOverrideInSimSpace=true
    bOverrideAngularDamping=true
    bUseSphericalLimits=true
    simulationSpace=RootRelative
    alphaBoolBlend {blendOption=Linear}
    boundBone {boneName=dyn_jacket_bk_r_02}
    componentAppliedLinearAccClamp=<3500 3500 3500>
    componentLinearAccScale=<1 1 1>
    componentLinearVelScale=<1 1 1>
    gravityOverride {z=-810}
    physicsBodyDefinitions [{
      sphereCollisionRadius=0
      boundBone {boneName=dyn_jacket_bk_r_02}
      boxExtents=<20 20 20>
      constraintSetup {angularLimitsMax=<5 5 0>, angularLimitsMin=<-5 -5 -45>, angularTarget=<0 0 0>}
      localJointOffset=<2 0 0>
     }]
    relativeSpaceBone {boneName=root}
    sphericalLimits [
     {limitRadius=15, drivingBone {boneName=pelvis}, sphereLocalOffset=<0 1 2>}
     {limitRadius=12, drivingBone {boneName=thigh_r}, sphereLocalOffset=<5 0 0>}
    ]
   }
   n34 = Anim Dynamics - Bone: dyn_jacket_bk_r_03(ComponentPose=@n33, Alpha=@n25, ExternalForce="0,0,0") {
    angularDampingOverride=0.99
    angularSpringConstant=50
    bAngularSpring=true
    bGravityOverrideInSimSpace=true
    bOverrideAngularDamping=true
    bUseSphericalLimits=true
    simulationSpace=RootRelative
    alphaBoolBlend {blendOption=Linear}
    boundBone {boneName=dyn_jacket_bk_r_03}
    componentAppliedLinearAccClamp=<3500 3500 3500>
    componentLinearAccScale=<1 1 1>
    componentLinearVelScale=<1 1 1>
    gravityOverride {z=-810}
    physicsBodyDefinitions [{
      sphereCollisionRadius=0
      boundBone {boneName=dyn_jacket_bk_r_03}
      boxExtents=<20 20 20>
      constraintSetup {angularLimitsMax=<5 5 0>, angularLimitsMin=<-5 -5 -45>, angularTarget=<0 0 0>}
      localJointOffset=<2 0 0>
     }]
    relativeSpaceBone {boneName=root}
    sphericalLimits [
     {limitRadius=15.5, drivingBone {boneName=pelvis}, sphereLocalOffset=<0 1 2>}
     {limitRadius=12, drivingBone {boneName=thigh_r}, sphereLocalOffset=<5 0 0>}
    ]
   }
   n35 = Get spine rotate() -> (spine rotate:double)
   n36 = "Transform (Modify) Bone - Bone: dyn_jacket_r_MSTR"!(ComponentPose=@n34, Alpha:float=@n35, Translation="0,0,0", Rotation="0,-15,0", Scale="1,1,1") -> (Pose) {
    rotationMode=BMM_Additive
    rotationSpace=BCS_BoneSpace
    translationMode=BMM_Additive
    alphaBoolBlend {blendInTime=0.1, blendOption=Linear, blendOutTime=0.3}
    alphaScaleBiasClamp {interpSpeedIncreasing=2}
    boneToModify {boneName=dyn_jacket_r_MSTR}
    rotation {yaw=-15}
   }
   n37 = "Transform (Modify) Bone - Bone: dyn_jacket_fr_r_MSTR"!(ComponentPose=@n36, Alpha:float=@n35, Translation="0,0,0", Rotation="0,15,0", Scale="1,1,1") -> (Pose) {
    rotationMode=BMM_Additive
    rotationSpace=BCS_BoneSpace
    translationMode=BMM_Additive
    alphaBoolBlend {blendInTime=0.1, blendOption=Linear, blendOutTime=0.3}
    alphaScaleBiasClamp {interpSpeedIncreasing=2}
    boneToModify {boneName=dyn_jacket_fr_r_MSTR}
    rotation {yaw=15}
   }
   n38 = "Transform (Modify) Bone - Bone: dyn_jacket_l_MSTR"!(ComponentPose=@n37, Alpha:float=@n35, Translation="0,0,0", Rotation="0,-15,0", Scale="1,1,1") -> (Pose) {
    rotationMode=BMM_Additive
    rotationSpace=BCS_BoneSpace
    translationMode=BMM_Additive
    alphaBoolBlend {blendInTime=0.1, blendOption=Linear, blendOutTime=0.3}
    alphaScaleBiasClamp {interpSpeedIncreasing=2}
    boneToModify {boneName=dyn_jacket_l_MSTR}
    rotation {yaw=-15}
   }
   n39 = "Transform (Modify) Bone - Bone: dyn_jacket_fr_l_MSTR"!(ComponentPose=@n38, Alpha:float=@n35, Translation="0,0,0", Rotation="0,-15,0.001", Scale="1,1,1") -> (Pose) {
    rotationMode=BMM_Additive
    rotationSpace=BCS_BoneSpace
    translationMode=BMM_Additive
    alphaBoolBlend {blendInTime=0.1, blendOption=Linear, blendOutTime=0.3}
    alphaScaleBiasClamp {interpSpeedIncreasing=2}
    boneToModify {boneName=dyn_jacket_fr_l_MSTR}
    rotation {roll=0.001, yaw=-15}
   }
   n40 = "Transform (Modify) Bone - Bone: dyn_jacket_bk_MSTR"!(ComponentPose=@n39, Alpha:float=@n35, Translation="0,0,0", Rotation="0,-25,0", Scale="1,1,1") -> (Pose) {
    rotationMode=BMM_Additive
    rotationSpace=BCS_BoneSpace
    translationMode=BMM_Additive
    alphaBoolBlend {blendInTime=0.1, blendOption=Linear, blendOutTime=0.3}
    alphaScaleBiasClamp {interpSpeedIncreasing=2}
    boneToModify {boneName=dyn_jacket_bk_MSTR}
    rotation {yaw=-25}
   }
   n41 = "Transform (Modify) Bone - Bone: dyn_jacket_bk_l_MSTR"!(ComponentPose=@n40, Alpha:float=@n35, Translation="0,0,0", Rotation="25,0,0", Scale="1,1,1") -> (Pose) {
    rotationMode=BMM_Additive
    rotationSpace=BCS_BoneSpace
    translationMode=BMM_Additive
    alphaBoolBlend {blendInTime=0.1, blendOption=Linear, blendOutTime=0.3}
    alphaScaleBiasClamp {interpSpeedIncreasing=2}
    boneToModify {boneName=dyn_jacket_bk_l_MSTR}
    rotation {pitch=25}
   }
   n42 = "Transform (Modify) Bone - Bone: dyn_jacket_bk_r_MSTR"!(ComponentPose=@n41, Alpha:float=@n35, Translation="0,0,0", Rotation="0,-25,0", Scale="1,1,1") -> (Pose) {
    rotationMode=BMM_Additive
    rotationSpace=BCS_BoneSpace
    translationMode=BMM_Additive
    alphaBoolBlend {blendInTime=0.1, blendOption=Linear, blendOutTime=0.3}
    alphaScaleBiasClamp {interpSpeedIncreasing=2}
    boneToModify {boneName=dyn_jacket_bk_r_MSTR}
    rotation {yaw=-25}
   }
   n43 = Get bIsInFrontEnd()
   n44 = "Transform (Modify) Bone - Bone: dyn_jacket_r_MSTR"!(ComponentPose=@n42, bAlphaBoolEnabled:bool=@n43, Translation="0,0,0", Rotation="0,-2,0", Scale="1,1,1") -> (Pose) {
    alphaInputType=Bool
    rotationMode=BMM_Additive
    rotationSpace=BCS_BoneSpace
    translationMode=BMM_Additive
    alphaBoolBlend {blendInTime=0.1, blendOption=Linear, blendOutTime=0.3}
    alphaScaleBiasClamp {bInterpResult=true, interpSpeedIncreasing=2}
    boneToModify {boneName=dyn_jacket_r_MSTR}
    rotation {yaw=-2}
   }
   n45 = "Transform (Modify) Bone - Bone: dyn_jacket_fr_r_MSTR"!(ComponentPose=@n44, bAlphaBoolEnabled:bool=@n43, Translation="0,0,0", Rotation="2,0,0", Scale="1,1,1") -> (Pose) {
    alphaInputType=Bool
    rotationMode=BMM_Additive
    rotationSpace=BCS_BoneSpace
    translationMode=BMM_Additive
    alphaBoolBlend {blendInTime=0.1, blendOption=Linear, blendOutTime=0.3}
    alphaScaleBiasClamp {bInterpResult=true, interpSpeedIncreasing=2}
    boneToModify {boneName=dyn_jacket_fr_r_MSTR}
    rotation {pitch=2}
   }
   n46 = Component To Local(ComponentPose=@n45)
   n47 = Get bIsSkyDiving()
   n48 = Get GravityOverride()
   n49 = RigidBody(ComponentPose=Local To Component(LocalPose=Use cached pose 'presim'()), bAlphaBoolEnabled=NOT Boolean(A=@n47), OverrideWorldGravity=@n48) {
    alphaInputType=Bool
    bClampLinearTranslationLimitToRefPose=true
    bOverrideWorldGravity=true
    overridePhysicsAsset=&/Pencil_BlazerCuff_MedB_Ubdy_Physics
    simulationSpace=BaseBoneSpace
    alphaBoolBlend {blendInTime=0.1, blendOption=Linear, blendOutTime=0.1}
    baseBoneRef {boneName=pelvis}
    componentAppliedLinearAccClamp=<3000 3000 3000>
    componentLinearAccScale=<0.3 0.3 0.5>
    componentLinearVelScale=<0.3 0.3 0.5>
    simSpaceSettings {
     maxAngularAcceleration=1000
     maxAngularVelocity=3
     maxLinearAcceleration=700
     maxLinearVelocity=500
     velocityScaleZ=0
     worldAlpha=1
     externalLinearDragV=<0.2 0.2>
    }
   }
   n50 = RigidBody(ComponentPose=@n49, bAlphaBoolEnabled=@n47, OverrideWorldGravity=@n48) {
    alphaInputType=Bool
    bClampLinearTranslationLimitToRefPose=true
    bOverrideWorldGravity=true
    overridePhysicsAsset=&Pencil_BlazerCuff_MedB_Ubdy_Physics
    simulationSpace=BaseBoneSpace
    alphaBoolBlend {blendInTime=0.1, blendOption=Linear, blendOutTime=0.1}
    baseBoneRef {boneName=pelvis}
    componentAppliedLinearAccClamp=<3000 3000 3000>
    componentLinearAccScale=<0.3 0.3 0.3>
    componentLinearVelScale=<0.3 0.3 0.3>
    simSpaceSettings {
     maxAngularAcceleration=1000
     maxAngularVelocity=3
     maxLinearAcceleration=700
     maxLinearVelocity=500
     velocityScaleZ=0
     externalLinearDragV=<0.2 0.2>
    }
   }
   n51 = Get spine rotate()
   n52 = "Transform (Modify) Bone - Bone: dyn_jacket_r_MSTR"!(ComponentPose=@n50, Alpha:float=@n51, Translation="0,0,0", Rotation="0,-15,0", Scale="1,1,1") -> (Pose) {
    rotationMode=BMM_Additive
    rotationSpace=BCS_BoneSpace
    translationMode=BMM_Additive
    alphaBoolBlend {blendInTime=0.1, blendOption=Linear, blendOutTime=0.3}
    alphaScaleBiasClamp {interpSpeedIncreasing=2}
    boneToModify {boneName=dyn_jacket_r_MSTR}
    rotation {yaw=-15}
   }
   n53 = "Transform (Modify) Bone - Bone: dyn_jacket_fr_r_MSTR"!(ComponentPose=@n52, Alpha:float=@n51, Translation="0,0,0", Rotation="0,15,0", Scale="1,1,1") -> (Pose) {
    rotationMode=BMM_Additive
    rotationSpace=BCS_BoneSpace
    translationMode=BMM_Additive
    alphaBoolBlend {blendInTime=0.1, blendOption=Linear, blendOutTime=0.3}
    alphaScaleBiasClamp {interpSpeedIncreasing=2}
    boneToModify {boneName=dyn_jacket_fr_r_MSTR}
    rotation {yaw=15}
   }
   n54 = "Transform (Modify) Bone - Bone: dyn_jacket_l_MSTR"(ComponentPose=@n53, Alpha=@n51, Translation="0,0,0", Rotation="0,-15,0", Scale="1,1,1") {
    rotationMode=BMM_Additive
    rotationSpace=BCS_BoneSpace
    translationMode=BMM_Additive
    alphaBoolBlend {blendInTime=0.1, blendOption=Linear, blendOutTime=0.3}
    alphaScaleBiasClamp {interpSpeedIncreasing=2}
    boneToModify {boneName=dyn_jacket_l_MSTR}
    rotation {yaw=-15}
   }
   n55 = "Transform (Modify) Bone - Bone: dyn_jacket_fr_l_MSTR"(ComponentPose=@n54, Alpha=@n51, Translation="0,0,0", Rotation="0,-15,0.001", Scale="1,1,1") {
    rotationMode=BMM_Additive
    rotationSpace=BCS_BoneSpace
    translationMode=BMM_Additive
    alphaBoolBlend {blendInTime=0.1, blendOption=Linear, blendOutTime=0.3}
    alphaScaleBiasClamp {interpSpeedIncreasing=2}
    boneToModify {boneName=dyn_jacket_fr_l_MSTR}
    rotation {roll=0.001, yaw=-15}
   }
   n56 = "Transform (Modify) Bone - Bone: dyn_jacket_bk_MSTR"(ComponentPose=@n55, Alpha=@n51, Translation="0,0,0", Rotation="0,-25,0", Scale="1,1,1") {
    rotationMode=BMM_Additive
    rotationSpace=BCS_BoneSpace
    translationMode=BMM_Additive
    alphaBoolBlend {blendInTime=0.1, blendOption=Linear, blendOutTime=0.3}
    alphaScaleBiasClamp {interpSpeedIncreasing=2}
    boneToModify {boneName=dyn_jacket_bk_MSTR}
    rotation {yaw=-25}
   }
   n57 = "Transform (Modify) Bone - Bone: dyn_jacket_bk_l_MSTR"(ComponentPose=@n56, Alpha=@n51, Translation="0,0,0", Rotation="25,0,0", Scale="1,1,1") {
    rotationMode=BMM_Additive
    rotationSpace=BCS_BoneSpace
    translationMode=BMM_Additive
    alphaBoolBlend {blendInTime=0.1, blendOption=Linear, blendOutTime=0.3}
    alphaScaleBiasClamp {interpSpeedIncreasing=2}
    boneToModify {boneName=dyn_jacket_bk_l_MSTR}
    rotation {pitch=25}
   }
   n58 = "Transform (Modify) Bone - Bone: dyn_jacket_bk_r_MSTR"(ComponentPose=@n57, Alpha=@n51, Translation="0,0,0", Rotation="0,-25,0", Scale="1,1,1") {
    rotationMode=BMM_Additive
    rotationSpace=BCS_BoneSpace
    translationMode=BMM_Additive
    alphaBoolBlend {blendInTime=0.1, blendOption=Linear, blendOutTime=0.3}
    alphaScaleBiasClamp {interpSpeedIncreasing=2}
    boneToModify {boneName=dyn_jacket_bk_r_MSTR}
    rotation {yaw=-25}
   }
   n59 = Component To Local(ComponentPose=@n58)
   n60 = Get Skirt Long() -> (Skirt Long:bool)
   n61 = Get bIsSkyDiving()
   n62 = Get GravityOverride()
   n63 = RigidBody(ComponentPose=Local To Component(LocalPose=Use cached pose 'presim'()), bAlphaBoolEnabled=NOT Boolean(A=@n61), OverrideWorldGravity=@n62) {
    alphaInputType=Bool
    bClampLinearTranslationLimitToRefPose=true
    bOverrideWorldGravity=true
    overridePhysicsAsset=&Pencil_BlazerCuff_MedB_Ubdy_ShortSkirt_Physics
    simulationSpace=BaseBoneSpace
    alphaBoolBlend {blendInTime=0.1, blendOption=Linear, blendOutTime=0.1}
    baseBoneRef {boneName=pelvis}
    componentAppliedLinearAccClamp=<3000 3000 3000>
    componentLinearAccScale=<0.3 0.3 0.5>
    componentLinearVelScale=<0.3 0.3 0.5>
    simSpaceSettings {
     maxAngularAcceleration=1000
     maxAngularVelocity=3
     maxLinearAcceleration=700
     maxLinearVelocity=500
     velocityScaleZ=0
     worldAlpha=1
     externalLinearDragV=<0.2 0.2>
    }
   }
   n64 = RigidBody(ComponentPose=@n63, bAlphaBoolEnabled=@n61, OverrideWorldGravity=@n62) {
    alphaInputType=Bool
    bClampLinearTranslationLimitToRefPose=true
    bOverrideWorldGravity=true
    overridePhysicsAsset=&Pencil_BlazerCuff_MedB_Ubdy_ShortSkirt_Physics
    simulationSpace=BaseBoneSpace
    alphaBoolBlend {blendInTime=0.1, blendOption=Linear, blendOutTime=0.1}
    baseBoneRef {boneName=pelvis}
    componentAppliedLinearAccClamp=<3000 3000 3000>
    componentLinearAccScale=<0.3 0.3 0.4>
    componentLinearVelScale=<0.3 0.3 0.4>
    simSpaceSettings {
     maxAngularAcceleration=1000
     maxAngularVelocity=3
     maxLinearAcceleration=700
     maxLinearVelocity=500
     velocityScaleZ=0
     externalLinearDragV=<0.2 0.2>
    }
   }
   Output Pose(Result=Blend Poses by bool(
     BlendPose_0=Component To Local(ComponentPose=@n64)
     BlendPose_1=Blend Poses by bool(BlendPose_0=@n46, BlendPose_1=@n59, bActiveValue=@n60) {blendType=Linear}
     bActiveValue=Get Skirt short() -> (Skirt short:bool)
    ) {blendType=Linear}) {layerGroup=None}
   n65 = Control Rig(Source=Copy Pose From Mesh() {bCopyCurves=true, bUseAttachedParent=true}) {
    alphaBoolBlend {blendOption=Linear}
    controlRigAssetReference {blueprintRigClass=&/Pencil_BlazerCuff_MedB_UBody_DeformRig.Pencil_BlazerCuff_MedB_UBody_DeformRig_C}
   }
   n66 = "Constraint - dyn_jacket_fr_l_Aim (PPPPP)"!(
    ComponentPose=Local To Component(LocalPose=@n65)
    Alpha=1
    ConstraintWeights_0=0.6001
    ConstraintWeights_1:float=0.1168
    ConstraintWeights_2:float=0.2217
    ConstraintWeights_3:float=0.0272
    ConstraintWeights_4:float=0.0342
   ) -> (Pose) {
    alphaBoolBlend {blendOption=Linear}
    boneToModify {boneName=dyn_jacket_fr_l_Aim}
    constraintSetup [
     {targetBone {boneName=spine_04}}
     {targetBone {boneName=spine_03}}
     {targetBone {boneName=spine_02}}
     {targetBone {boneName=spine_01}}
     {targetBone {boneName=pelvis}}
    ]
    constraintWeights [1,1,1,1,1]
   }
   n67 = "Constraint - dyn_jacket_l_Aim (PPPPPPPP)"!(
    ComponentPose=@n66
    Alpha=1
    ConstraintWeights_0=0.0084
    ConstraintWeights_1:float=0.0077
    ConstraintWeights_2:float=0.406
    ConstraintWeights_3:float=0.0036
    ConstraintWeights_4:float=0.1004
    ConstraintWeights_5:float=0.3749
    ConstraintWeights_6:float=0.0205
    ConstraintWeights_7:float=0.0785
   ) -> (Pose) {
    alphaBoolBlend {blendOption=Linear}
    boneToModify {boneName=dyn_jacket_l_Aim}
    constraintSetup [
     {targetBone {boneName=spine_05}}
     {targetBone {boneName=clavicle_l}}
     {targetBone {boneName=spine_04}}
     {targetBone {boneName=upperarm_twist_01_l}}
     {targetBone {boneName=spine_03}}
     {targetBone {boneName=spine_02}}
     {targetBone {boneName=spine_01}}
     {targetBone {boneName=pelvis}}
    ]
    constraintWeights [1,1,1,1,1,1,1,1]
   }
   n68 = "Constraint - dyn_jacket_bk_l_Aim (PPPPPPPP)"!(
    ComponentPose=@n67
    Alpha=1
    ConstraintWeights_0=0.0112
    ConstraintWeights_1:float=0.0259
    ConstraintWeights_2:float=0.3621
    ConstraintWeights_3:float=0.0019
    ConstraintWeights_4:float=0.1271
    ConstraintWeights_5:float=0.3926
    ConstraintWeights_6:float=0.046
    ConstraintWeights_7:float=0.0332
   ) -> (Pose) {
    alphaBoolBlend {blendOption=Linear}
    boneToModify {boneName=dyn_jacket_bk_l_Aim}
    constraintSetup [
     {targetBone {boneName=spine_05}}
     {targetBone {boneName=clavicle_l}}
     {targetBone {boneName=spine_04}}
     {targetBone {boneName=upperarm_twist_01_l}}
     {targetBone {boneName=spine_03}}
     {targetBone {boneName=spine_02}}
     {targetBone {boneName=spine_01}}
     {targetBone {boneName=pelvis}}
    ]
    constraintWeights [1,1,1,1,1,1,1,1]
   }
   n69 = "Constraint - dyn_jacket_bk_Aim (PPPPPPPP)"!(
    ComponentPose=@n68
    Alpha=1
    ConstraintWeights_0=0.0325
    ConstraintWeights_1:float=0.005
    ConstraintWeights_2:float=0.0063
    ConstraintWeights_3:float=0.2808
    ConstraintWeights_4:float=0.1449
    ConstraintWeights_5:float=0.3921
    ConstraintWeights_6:float=0.0549
    ConstraintWeights_7:float=0.0835
   ) -> (Pose) {
    alphaBoolBlend {blendOption=Linear}
    boneToModify {boneName=dyn_jacket_bk_Aim}
    constraintSetup [
     {targetBone {boneName=spine_05}}
     {targetBone {boneName=clavicle_l}}
     {targetBone {boneName=clavicle_r}}
     {targetBone {boneName=spine_04}}
     {targetBone {boneName=spine_03}}
     {targetBone {boneName=spine_02}}
     {targetBone {boneName=spine_01}}
     {targetBone {boneName=pelvis}}
    ]
    constraintWeights [1,1,1,1,1,1,1,1]
   }
   n70 = "Constraint - dyn_jacket_bk_r_Aim (PPPPPPPP)"!(
    ComponentPose=@n69
    Alpha=1
    ConstraintWeights_0=0.0073
    ConstraintWeights_1:float=0.0075
    ConstraintWeights_2:float=0.3579
    ConstraintWeights_3:float=0.0021
    ConstraintWeights_4:float=0.0792
    ConstraintWeights_5:float=0.3957
    ConstraintWeights_6:float=0.0358
    ConstraintWeights_7:float=0.1144
   ) -> (Pose) {
    alphaBoolBlend {blendOption=Linear}
    boneToModify {boneName=dyn_jacket_bk_r_Aim}
    constraintSetup [
     {targetBone {boneName=spine_05}}
     {targetBone {boneName=clavicle_r}}
     {targetBone {boneName=spine_04}}
     {targetBone {boneName=upperarm_twist_01_r}}
     {targetBone {boneName=spine_03}}
     {targetBone {boneName=spine_02}}
     {targetBone {boneName=spine_01}}
     {targetBone {boneName=pelvis}}
    ]
    constraintWeights [1,1,1,1,1,1,1,1]
   }
   n71 = "Constraint - dyn_jacket_r_Aim (PPPPPPPP)"!(
    ComponentPose=@n70
    Alpha=1
    ConstraintWeights_0=0.0041
    ConstraintWeights_1:float=0.0037
    ConstraintWeights_2:float=0.4076
    ConstraintWeights_3:float=0.0013
    ConstraintWeights_4:float=0.0733
    ConstraintWeights_5:float=0.3819
    ConstraintWeights_6:float=0.0135
    ConstraintWeights_7:float=0.1145
   ) -> (Pose) {
    alphaBoolBlend {blendOption=Linear}
    boneToModify {boneName=dyn_jacket_r_Aim}
    constraintSetup [
     {targetBone {boneName=spine_05}}
     {targetBone {boneName=clavicle_r}}
     {targetBone {boneName=spine_04}}
     {targetBone {boneName=upperarm_twist_01_r}}
     {targetBone {boneName=spine_03}}
     {targetBone {boneName=spine_02}}
     {targetBone {boneName=spine_01}}
     {targetBone {boneName=pelvis}}
    ]
    constraintWeights [1,1,1,1,1,1,1,1]
   }
   n72 = "Constraint - dyn_jacket_fr_r_Aim (PPPPP)"!(
    ComponentPose=@n71
    Alpha=1
    ConstraintWeights_0=0.5695
    ConstraintWeights_1:float=0.1181
    ConstraintWeights_2:float=0.2424
    ConstraintWeights_3:float=0.0275
    ConstraintWeights_4:float=0.0426
   ) -> (Pose) {
    alphaBoolBlend {blendOption=Linear}
    boneToModify {boneName=dyn_jacket_fr_r_Aim}
    constraintSetup [
     {targetBone {boneName=spine_04}}
     {targetBone {boneName=spine_03}}
     {targetBone {boneName=spine_02}}
     {targetBone {boneName=spine_01}}
     {targetBone {boneName=pelvis}}
    ]
    constraintWeights [1,1,1,1,1]
   }
   n73 = "Constraint - dyn_Cloth_volume_r_C (PPP)"!(ComponentPose=@n72, Alpha=1, ConstraintWeights_0=0.3367, ConstraintWeights_1:float=0.256, ConstraintWeights_2:float=0.4073) -> (Pose) {
    alphaBoolBlend {blendOption=Linear}
    boneToModify {boneName=dyn_Cloth_volume_r_C}
    constraintSetup [{targetBone {boneName=pelvis}}, {targetBone {boneName=thigh_twist_01_r}}, {targetBone {boneName=deform_glute_r}}]
    constraintWeights [1,1,1]
   }
   n74 = "Constraint - dyn_Cloth_volume_bk (PPPPP)"!(
    ComponentPose=@n73
    Alpha=1
    ConstraintWeights_0=0.7511
    ConstraintWeights_1:float=0.0501
    ConstraintWeights_2:float=0.0504
    ConstraintWeights_3:float=0.0735
    ConstraintWeights_4:float=0.0749
   ) -> (Pose) {
    alphaBoolBlend {blendOption=Linear}
    boneToModify {boneName=dyn_Cloth_volume_bk}
    constraintSetup [
     {targetBone {boneName=pelvis}}
     {targetBone {boneName=thigh_twist_01_l}}
     {targetBone {boneName=thigh_twist_01_r}}
     {targetBone {boneName=deform_glute_l}}
     {targetBone {boneName=deform_glute_r}}
    ]
    constraintWeights [1,1,1,1,1]
   }
   n75 = "Constraint - dyn_Cloth_volume_l_C (PPP)"!(ComponentPose=@n74, Alpha=1, ConstraintWeights_0=0.3367, ConstraintWeights_1:float=0.256, ConstraintWeights_2:float=0.4073) -> (Pose) {
    alphaBoolBlend {blendOption=Linear}
    boneToModify {boneName=dyn_Cloth_volume_l_C}
    constraintSetup [{targetBone {boneName=pelvis}}, {targetBone {boneName=thigh_twist_01_l}}, {targetBone {boneName=deform_glute_l}}]
    constraintWeights [1,1,1]
   }
   n76 = "Constraint - dyn_Cloth_volume_r_B (PPPPP)"!(
    ComponentPose=@n75
    Alpha=1
    ConstraintWeights_0=0.1783
    ConstraintWeights_1:float=0.6512
    ConstraintWeights_2:float=0.1127
    ConstraintWeights_3:float=0.0301
    ConstraintWeights_4:float=0.0276
   ) -> (Pose) {
    alphaBoolBlend {blendOption=Linear}
    boneToModify {boneName=dyn_Cloth_volume_r_B}
    constraintSetup [
     {targetBone {boneName=pelvis}}
     {targetBone {boneName=thigh_twist_01_r}}
     {targetBone {boneName=thigh_r}}
     {targetBone {boneName=deform_glute_r}}
     {targetBone {boneName=deform_groin_r}}
    ]
    constraintWeights [1,1,1,1,1]
   }
   n77 = "Constraint - dyn_Cloth_volume_l_B (PPPPP)"!(
    ComponentPose=@n76
    Alpha=1
    ConstraintWeights_0=0.1783
    ConstraintWeights_1:float=0.6512
    ConstraintWeights_2:float=0.1127
    ConstraintWeights_3:float=0.0301
    ConstraintWeights_4:float=0.0276
   ) -> (Pose) {
    alphaBoolBlend {blendOption=Linear}
    boneToModify {boneName=dyn_Cloth_volume_l_B}
    constraintSetup [
     {targetBone {boneName=pelvis}}
     {targetBone {boneName=thigh_twist_01_l}}
     {targetBone {boneName=thigh_l}}
     {targetBone {boneName=deform_glute_l}}
     {targetBone {boneName=deform_groin_l}}
    ]
    constraintWeights [1,1,1,1,1]
   }
   n78 = "Constraint - dyn_Cloth_volume_r_A (PPPP)"!(
    ComponentPose=@n77
    Alpha=1
    ConstraintWeights_0=0.2372
    ConstraintWeights_1:float=0.4317
    ConstraintWeights_2:float=0.0623
    ConstraintWeights_3:float=0.2688
   ) -> (Pose) {
    alphaBoolBlend {blendOption=Linear}
    boneToModify {boneName=dyn_Cloth_volume_r_A}
    constraintSetup [
     {targetBone {boneName=pelvis}}
     {targetBone {boneName=thigh_twist_01_r}}
     {targetBone {boneName=thigh_r}}
     {targetBone {boneName=deform_groin_r}}
    ]
    constraintWeights [1,1,1,1]
   }
   n79 = "Constraint - dyn_Cloth_volume_fn (PPPPP)"!(
    ComponentPose=@n78
    Alpha=1
    ConstraintWeights_0=0.5182
    ConstraintWeights_1:float=0.1769
    ConstraintWeights_2:float=0.1714
    ConstraintWeights_3:float=0.0703
    ConstraintWeights_4:float=0.0631
   ) -> (Pose) {
    alphaBoolBlend {blendOption=Linear}
    boneToModify {boneName=dyn_Cloth_volume_fn}
    constraintSetup [
     {targetBone {boneName=pelvis}}
     {targetBone {boneName=thigh_twist_01_l}}
     {targetBone {boneName=thigh_twist_01_r}}
     {targetBone {boneName=deform_groin_l}}
     {targetBone {boneName=deform_groin_r}}
    ]
    constraintWeights [1,1,1,1,1]
   }
   n80 = "Constraint - dyn_Cloth_volume_l_A (PPPP)"!(
    ComponentPose=@n79
    Alpha=1
    ConstraintWeights_0=0.2372
    ConstraintWeights_1:float=0.4317
    ConstraintWeights_2:float=0.0623
    ConstraintWeights_3:float=0.2688
   ) -> (Pose) {
    alphaBoolBlend {blendOption=Linear}
    boneToModify {boneName=dyn_Cloth_volume_l_A}
    constraintSetup [
     {targetBone {boneName=pelvis}}
     {targetBone {boneName=thigh_twist_01_l}}
     {targetBone {boneName=thigh_l}}
     {targetBone {boneName=deform_groin_l}}
    ]
    constraintWeights [1,1,1,1]
   }
   n81 = Look At - Bone: dyn_jacket_bk_Aim(ComponentPose=@n80, Alpha=1) {
    alphaBoolBlend {blendOption=Linear}
    boneToModify {boneName=dyn_jacket_bk_Aim}
    lookAtLocation {x=0}
    lookAtTarget {boneReference {boneName=dyn_jacket_TGT}}
    lookAt_Axis {axis=<1 0>}
   }
   n82 = Look At - Bone: dyn_jacket_bk_l_Aim(ComponentPose=@n81, Alpha=1) {
    alphaBoolBlend {blendOption=Linear}
    boneToModify {boneName=dyn_jacket_bk_l_Aim}
    lookAtLocation {x=0}
    lookAtTarget {boneReference {boneName=dyn_jacket_TGT}}
    lookAt_Axis {axis=<1 0>}
   }
   n83 = Look At - Bone: dyn_jacket_bk_r_Aim(ComponentPose=@n82, Alpha=1) {
    alphaBoolBlend {blendOption=Linear}
    boneToModify {boneName=dyn_jacket_bk_r_Aim}
    lookAtLocation {x=0}
    lookAtTarget {boneReference {boneName=dyn_jacket_TGT}}
    lookAt_Axis {axis=<1 0>}
   }
   n84 = Look At - Bone: dyn_jacket_l_Aim(ComponentPose=@n83, Alpha=1) {
    alphaBoolBlend {blendOption=Linear}
    boneToModify {boneName=dyn_jacket_l_Aim}
    lookAtLocation {x=0}
    lookAtTarget {boneReference {boneName=dyn_jacket_TGT}}
    lookAt_Axis {axis=<1 0>}
   }
   n85 = Look At - Bone: dyn_jacket_r_Aim(ComponentPose=@n84, Alpha=1) {
    alphaBoolBlend {blendOption=Linear}
    boneToModify {boneName=dyn_jacket_r_Aim}
    lookAtLocation {x=0}
    lookAtTarget {boneReference {boneName=dyn_jacket_TGT}}
    lookAt_Axis {axis=<1 0>}
   }
   n86 = Look At - Bone: dyn_jacket_fr_l_Aim(ComponentPose=@n85, Alpha=1) {
    alphaBoolBlend {blendOption=Linear}
    boneToModify {boneName=dyn_jacket_fr_l_Aim}
    lookAtLocation {x=0}
    lookAtTarget {boneReference {boneName=dyn_jacket_TGT}}
    lookAt_Axis {axis=<1 0>}
   }
   n87 = Look At - Bone: dyn_jacket_fr_r_Aim(ComponentPose=@n86, Alpha=1) {
    alphaBoolBlend {blendOption=Linear}
    boneToModify {boneName=dyn_jacket_fr_r_Aim}
    lookAtLocation {x=0}
    lookAtTarget {boneReference {boneName=dyn_jacket_TGT}}
    lookAt_Axis {axis=<1 0>}
   }
   Save cached pose 'Parts check'(Pose=Component To Local(ComponentPose=@n87)) {cacheName=Parts check}
   n88 = Control Rig!(
    Source=Use cached pose 'Parts check'()
    Alpha
    crouch:double=Get crouch() -> (crouch:double)
    moving:double=Get Moving 2D() -> (Moving 2D:double)
   ) -> (Pose) {
    alphaBoolBlend {blendOption=Linear}
    controlRigAssetReference {blueprintRigClass=&/Pencil_BlazerCuff_MedB_Ubdy_Perf_CtrlRig.Pencil_BlazerCuff_MedB_Ubdy_Perf_CtrlRig_C}
   }
   n89 = Control Rig!(
    Source=Use cached pose 'Parts check'()
    Alpha
    crouching:double=Get crouch()
    crouch strafe r:double=Get strafe l() -> (strafe l:double)
    crouch strafe l:double=Get strafe r() -> (strafe r:double)
    moving 2d:double=Get Moving 2D()
    emote:double=Get emote() -> (emote:double)
   ) -> (Pose) {
    alphaBoolBlend {blendOption=Linear}
    controlRigAssetReference {blueprintRigClass=/BRCosmetics/Characters/Player/Female/Medium/Bodies/F_MED_Pencil_B/SkirtLong/Meshes/Pencil_SkirtLong_MedB_Lbdy_CtrlRig.Pencil_SkirtLong_MedB_Lbdy_CtrlRig_C}
   }
   n90 = Control Rig!(
    Source=@n89
    Alpha
    crouch:double=Get crouch()
    moving:double=Get Moving 2D()
    PelvisDown:double=Get DetectedPelvisDown() -> (DetectedPelvisDown:double)
   ) -> (Pose) {
    alphaBoolBlend {blendOption=Linear}
    controlRigAssetReference {blueprintRigClass=&/Pencil_BlazerCuff_MedB_Ubdy_Perf_SkirtLong_CtrlRig.Pencil_BlazerCuff_MedB_Ubdy_Perf_SkirtLong_CtrlRig_C}
   }
   n91 = Blend Poses by bool(BlendPose_0=@n90, BlendPose_1=@n88, bActiveValue=Get Skirt Long()) {blendType=Linear}
   n92 = Control Rig!(
    Source=Use cached pose 'Parts check'()
    Alpha
    crouching:double=Get crouch()
    crouch strafe r:double=Get strafe l()
    crouch strafe l:double=Get strafe r()
    moving 2d:double=Get Moving 2D()
    emote:double=Get emote()
   ) -> (Pose) {
    alphaBoolBlend {blendOption=Linear}
    controlRigAssetReference {blueprintRigClass=&Pencil_SkirtLong_MedB_Lbdy_CtrlRig_C}
   }
   n93 = Control Rig!(Source=@n92, Alpha, crouch:double=Get crouch(), moving:double=Get Moving 2D()) -> (Pose) {
    alphaBoolBlend {blendOption=Linear}
    controlRigAssetReference {blueprintRigClass=&/Pencil_BlazerCuff_MedB_Ubdy_Perf_SkirtShort_CtrlRig.Pencil_BlazerCuff_MedB_Ubdy_Perf_SkirtShort_CtrlRig_C}
   }
   n94 = Get Skirt short()
   n95 = Get jumping() -> (jumping:double)
   n96 = "Transform (Modify) Bone - Bone: dyn_jacket_bk_MSTR"(
    ComponentPose=Local To Component(LocalPose=Blend Poses by bool(BlendPose_0=@n93, BlendPose_1=@n91, bActiveValue=@n94) {blendType=Linear})
    Alpha=@n95
    Translation="0,-2,0"
    Rotation="0,-25,0"
    Scale="1,1,1"
   ) {
    rotationMode=BMM_Additive
    rotationSpace=BCS_BoneSpace
    translationMode=BMM_Additive
    alphaBoolBlend {blendInTime=0.1, blendOption=Linear, blendOutTime=0.3}
    alphaScaleBiasClamp {bInterpResult=true, interpSpeedIncreasing=2}
    boneToModify {boneName=dyn_jacket_bk_MSTR}
    rotation {yaw=-25}
    translation {y=-2}
   }
   n97 = "Transform (Modify) Bone - Bone: dyn_jacket_bk_l_MSTR"(ComponentPose=@n96, Alpha=@n95, Translation="0,0,2", Rotation="25,0,0", Scale="1,1,1") {
    rotationMode=BMM_Additive
    rotationSpace=BCS_BoneSpace
    translationMode=BMM_Additive
    alphaBoolBlend {blendInTime=0.1, blendOption=Linear, blendOutTime=0.3}
    alphaScaleBiasClamp {bInterpResult=true, interpSpeedIncreasing=2}
    boneToModify {boneName=dyn_jacket_bk_l_MSTR}
    rotation {pitch=25}
    translation {z=2}
   }
   n98 = "Transform (Modify) Bone - Bone: dyn_jacket_bk_r_MSTR"(ComponentPose=@n97, Alpha=@n95, Translation="0,-2,0", Rotation="0,-25,0", Scale="1,1,1") {
    rotationMode=BMM_Additive
    rotationSpace=BCS_BoneSpace
    translationMode=BMM_Additive
    alphaBoolBlend {blendInTime=0.1, blendOption=Linear, blendOutTime=0.3}
    alphaScaleBiasClamp {bInterpResult=true, interpSpeedIncreasing=2}
    boneToModify {boneName=dyn_jacket_bk_r_MSTR}
    rotation {yaw=-25}
    translation {y=-2}
   }
   n99 = "Transform (Modify) Bone - Bone: dyn_jacket_fr_r_MSTR"(ComponentPose=@n98, Alpha=@n95, Translation="-7,-6,2", Rotation="0,0,0", Scale="1,1,1") {
    rotationMode=BMM_Additive
    rotationSpace=BCS_BoneSpace
    translationMode=BMM_Additive
    alphaBoolBlend {blendInTime=0.1, blendOption=Linear, blendOutTime=0.3}
    alphaScaleBiasClamp {bInterpResult=true}
    boneToModify {boneName=dyn_jacket_fr_r_MSTR}
    translation=<-7 -6 2>
   }
   n100 = Get bIsMoving2D()
   n101 = "Transform (Modify) Bone - Bone: dyn_jacket_bk_MSTR"!(ComponentPose=@n99, bAlphaBoolEnabled:bool=@n100, Translation="0,0,0", Rotation="0,-5,0", Scale="1,1,1") -> (Pose) {
    alphaInputType=Bool
    rotationMode=BMM_Additive
    rotationSpace=BCS_BoneSpace
    alphaBoolBlend {blendInTime=0.1, blendOption=Linear, blendOutTime=0.3}
    boneToModify {boneName=dyn_jacket_bk_MSTR}
    rotation {yaw=-5}
   }
   n102 = "Transform (Modify) Bone - Bone: dyn_jacket_bk_l_MSTR"!(ComponentPose=@n101, bAlphaBoolEnabled:bool=@n100, Translation="0,0,0", Rotation="5,0,0", Scale="1,1,1") -> (Pose) {
    alphaInputType=Bool
    rotationMode=BMM_Additive
    rotationSpace=BCS_BoneSpace
    alphaBoolBlend {blendInTime=0.1, blendOption=Linear, blendOutTime=0.3}
    boneToModify {boneName=dyn_jacket_bk_l_MSTR}
    rotation {pitch=5}
   }
   n103 = "Transform (Modify) Bone - Bone: dyn_jacket_bk_r_MSTR"!(ComponentPose=@n102, bAlphaBoolEnabled:bool=@n100, Translation="0,0,0", Rotation="0,-5,0", Scale="1,1,1") -> (Pose) {
    alphaInputType=Bool
    rotationMode=BMM_Additive
    rotationSpace=BCS_BoneSpace
    alphaBoolBlend {blendInTime=0.1, blendOption=Linear, blendOutTime=0.3}
    boneToModify {boneName=dyn_jacket_bk_r_MSTR}
    rotation {yaw=-5}
   }
   n104 = Get bIsSkyDiving()
   n105 = "Transform (Modify) Bone - Bone: dyn_jacket_bk_MSTR"(ComponentPose=@n103, bAlphaBoolEnabled=@n104, Translation="0,-2,0", Rotation="0,-20,0", Scale="1,1,1") {
    alphaInputType=Bool
    rotationMode=BMM_Additive
    rotationSpace=BCS_BoneSpace
    translationMode=BMM_Additive
    alphaBoolBlend {blendInTime=0.1, blendOption=Linear, blendOutTime=0.3}
    alphaScaleBiasClamp {bInterpResult=true, interpSpeedIncreasing=2}
    boneToModify {boneName=dyn_jacket_bk_MSTR}
    rotation {yaw=-20}
    translation {y=-2}
   }
   n106 = "Transform (Modify) Bone - Bone: dyn_jacket_bk_l_MSTR"(ComponentPose=@n105, bAlphaBoolEnabled=@n104, Translation="0,0,2", Rotation="20,-35,0", Scale="1,1,1") {
    alphaInputType=Bool
    rotationMode=BMM_Additive
    rotationSpace=BCS_BoneSpace
    translationMode=BMM_Additive
    alphaBoolBlend {blendInTime=0.1, blendOption=Linear, blendOutTime=0.3}
    alphaScaleBiasClamp {bInterpResult=true, interpSpeedIncreasing=2}
    boneToModify {boneName=dyn_jacket_bk_l_MSTR}
    rotation {pitch=20, yaw=-35}
    translation {z=2}
   }
   n107 = "Transform (Modify) Bone - Bone: dyn_jacket_bk_r_MSTR"(ComponentPose=@n106, bAlphaBoolEnabled=@n104, Translation="0,-2,0", Rotation="35,-20,0", Scale="1,1,1") {
    alphaInputType=Bool
    rotationMode=BMM_Additive
    rotationSpace=BCS_BoneSpace
    translationMode=BMM_Additive
    alphaBoolBlend {blendInTime=0.1, blendOption=Linear, blendOutTime=0.3}
    alphaScaleBiasClamp {bInterpResult=true, interpSpeedIncreasing=2}
    boneToModify {boneName=dyn_jacket_bk_r_MSTR}
    rotation {pitch=35, yaw=-20}
    translation {y=-2}
   }
   n108 = "Transform (Modify) Bone - Bone: dyn_shirt_arm_fix"!(ComponentPose=@n107, bAlphaBoolEnabled:bool=True, Translation="8.5,0,0", Rotation="0,0,0", Scale="0.5,0.5,0.5") -> (Pose) {
    alphaInputType=Bool
    rotationMode=BMM_Additive
    rotationSpace=BCS_BoneSpace
    scaleMode=BMM_Additive
    scaleSpace=BCS_BoneSpace
    translationMode=BMM_Additive
    translationSpace=BCS_BoneSpace
    alphaBoolBlend {blendInTime=0.1, blendOption=Linear, blendOutTime=0.3}
    alphaScaleBiasClamp {bInterpResult=true, interpSpeedIncreasing=2}
    boneToModify {boneName=dyn_shirt_arm_fix}
    scale=<0.5 0.5 0.5>
    translation {x=8.5}
   }
   Save cached pose 'presim'(Pose=Component To Local(ComponentPose=@n108)) {cacheName=presim}
   NOT Boolean(A=@n1)
   Get jumping()
  ]}
 EventGraph {nodes [
   Event BlueprintUpdateAnimation()
   > n1 = Sequence!(execute) -> (then_0, then_1, then_2:Exec, then_3:Exec, then_4:Exec)
   > n2 = Gravity Override State Function DEPRECATED(CharacterPartAnimInstance=Self-Reference(), GameTimeInSeconds=GetGameTimeInSeconds(), GravityOverrideParams=Get Gravity Override Params())
   > Set GravityOverride(GravityOverride=@n2)
   n3 = Get bIsMoving2D()
   > n4 = Branch(Condition=AND Boolean(A=Get bIsCrouching(), B=NOT Boolean(A=@n3)))
   > Set crouch_still(crouch_still:bool=true) -> (Output_Get:bool)
   Set crouch_still(execute=@n4.else, crouch_still=false)
   n5 = PlayerMovement(execute=@n1.then_1, AnimInstance=Self-Reference(), XDotProductCoefficient=1, YDotProductCoefficient=1, ZDotProductCoefficient=1)
   > Set strafe r(strafe r:double=MapRangeClamped(Value=@n5, InRangeA=0, InRangeB=1)) -> (Output_Get:double)
   > Set strafe l(strafe l:double=MapRangeClamped(Value=@n5, InRangeA=0, InRangeB=-1)) -> (Output_Get:double)
   > n6 = Branch(Condition=Get bIsMoving2D())
   > Set Moving 2D(Moving 2D:double=1) -> (Output_Get:double)
   Set Moving 2D(execute=@n6.else, Moving 2D=0)
   Set spine rotate(
    execute=[@n1.then_2, @n1.then_3]
    spine rotate:double=GetAlphaFromReferenceTransformDelta(SkeletalMeshComponent=Get Skel_Mesh(), BoneName=spine_05, BaseName=spine_01, InputRange=(X=-10,Y=-30), TwistAxis="1,0,0")
   ) -> (Output_Get:double)
   > n7 = Branch(Condition=Get bIsSkyDiving())
   > Set skydive(skydive:double=0.5) -> (Output_Get:double)
   > Set skydive_anim dynamics(skydive_anim dynamics:double=0.5) -> (Output_Get:double)
   Set skydive(execute=@n7.else, skydive=0)
   > Set skydive_anim dynamics(skydive_anim dynamics=1)
   > Set spine rotate(spine rotate=0)
   n8 = GetOwnerSkeletalMeshComponent()
   n9 = BreakVector(InVec=BreakTransform(InTransform=GetDeltaTransformFromRefPose(self=@n8, BoneName=pelvis, BaseName=root)))
   n10 = Branch(execute=@n1.then_4, Condition=AND Boolean(A="float < float"(A=@n9, B=-60), B=Get bIsPlayingEmote()))
   > Set DetectedPelvisDown(DetectedPelvisDown:double=1) -> (Output_Get:double)
   Set DetectedPelvisDown(execute=@n10.else, DetectedPelvisDown=0)
   Event OnCharacterCustomizationLoaded()
   > n11 = Sequence!(execute) -> (then_0, then_1)
   > n12 = Branch(Condition=CharacterItemDefinitionHasAnyMetaTag(Tags=Make Array("[0]"=(TagName="Cosmetics.MetaTags.Pencil.Lower.SkirtLong"))))
   > Set Skirt Long(Skirt Long:bool=true) -> (Output_Get:bool)
   Set Skirt Long(execute=@n12.else, Skirt Long=false)
   n13 = Branch(
    execute=@n11.then_1
    Condition=CharacterItemDefinitionHasAnyMetaTag(Tags=Make Array("[0]"=(TagName="Cosmetics.MetaTags.Pencil.Lower.SkirtShort")))
   )
   > Set Skirt short(Skirt short:bool=true) -> (Output_Get:bool)
   Set Skirt short(execute=@n13.else, Skirt short=false)
   Event OnBeginPlayingEmote() {is=title K2Node_Event}
   > Set emote(emote:double=1) -> (Output_Get:double)
   Event OnEndPlayingEmote() {is=title K2Node_Event}
   > Set emote(emote=0)
   n14 = Event OnEndFalling()
   Set jumping(execute=[@n14.then, @n16.then], jumping:double=0) -> (Output_Get:double)
   n15 = Event OnBeginJumping()
   Set jumping(execute=[@n15.then, @n17.then], jumping=1)
   Event BlueprintInitializeAnimation()
   > Set Skel_Mesh(Skel_Mesh=GetOwnerSkeletalMeshComponent())
   n16 = Event OnEndJumping()
   Event OnEndCrouching()
   > Set crouch(crouch:double=0) -> (Output_Get:double)
   n17 = Event OnBeginFalling()
   Event OnBeginCrouching()
   > Set crouch(crouch=1)
  ]}
]
variables [
 crouch {type=number}
 crouch_still {type=boolean}
 DetectedPelvisDown {type=number}
 emote {type=number}
 Gravity Override Params {type=GravityOverrideParamsStruct, default {skydive_FloatingDown {gravityOverride=<0 -350 300>}}}
 GravityOverride {type=Vector}
 jumping {type=number}
 Moving 2D {type=number}
 Skel_Mesh {type=SkeletalMeshComponent}
 Skirt Long {type=boolean}
 Skirt short {type=boolean}
 skydive {type=number}
 skydive_anim dynamics {type=number}
 spine rotate {type=number}
 strafe l {type=number}
 strafe r {type=number}
]
